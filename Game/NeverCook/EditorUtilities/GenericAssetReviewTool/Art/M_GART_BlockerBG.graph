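# Translucent UI material for the blocker background in the editor-only Generic Asset Review Tool, drawing a pulsing, edge-weighted overlay.
# A time-driven LinearSine with a Period parameter blends between the Color0 and Color1 parameters, red and salmon by default.
# An inverted MF_SquareEdgeMask, controlled by Size, is multiplied by a vertical UV gradient to form a border mask that fades toward the bottom.
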Material /Game/NeverCook/EditorUtilities/GenericAssetReviewTool/Art/M_GART_BlockerBG
graph {
 properties {
  blendMode=BLEND_Translucent
  materialDomain=MD_UI
  opacityMaskClipValue=0.01
  pixelDepthOffsetMode=PDOM_Legacy
  usedShadingModels=Default Lit
 }
 nodes [
  LinearInterpolate(
   A=VectorParameter() {parameterName=Color0, defaultValue=<rgba 0.4115 0 0 1>}.RGB
   B=VectorParameter() {parameterName=Color1, defaultValue=<rgba 1 0.3062 0.1763 1>}.RGB
   Alpha=MaterialFunctionCall(Value=Time(), Period=ScalarParameter() {defaultValue=2, parameterName=Period}) {materialFunction=&LinearSine}.Linear Sine
  )
  n1 = TextureCoordinate()
  n2 = ScalarParameter() {defaultValue=10, parameterName=Size}
  Multiply(
   A=OneMinus(None=MaterialFunctionCall!(UVs=@n1, Mult=@n2, Exp) {materialFunction=/Game/Effects/Fort_Effects/MaterialFunctions/MF_SquareEdgeMask}.SquareMask)
   B=InverseLinearInterpolate(Value=ComponentMask(None=@n1) {g=true}) {constA=0.95, constB=0}
  )
 ]
}
