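# M_GART_DownArrowSelect is a translucent UI material in the editor-only Generic Asset Review Tool, named as the selected down-arrow graphic.
# Its graph holds a sample of the FTUE_Arrow_Triangle texture, borrowed from the Athena HUD tutorial art, and a constant of one.
# No links from these nodes to the material outputs are shown.
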Material /Game/NeverCook/EditorUtilities/GenericAssetReviewTool/Art/M_GART_DownArrowSelect
graph {
 properties {
  blendMode=BLEND_Translucent
  materialDomain=MD_UI
  opacityMaskClipValue=0.01
  pixelDepthOffsetMode=PDOM_Legacy
  usedShadingModels=Default Lit
 }
 nodes [
  Constant() {r=1}
  TextureSample() {texture=/Game/Athena/HUD/Art/Tutorial/FTUE_Arrow_Triangle}
 ]
}
