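# UI material for list item backgrounds in the Generic Asset Review Tool editor utility, showing a Color tint with a highlight for selected items.
# A procedural UV mask combines left and right edge falloffs, faded in from top to bottom, with a band along the bottom, then shapes it with an exponent.
# The mask blends the tint toward a brightened version of itself, and the Selected scalar fades that highlight in over the plain color.
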Material /Game/NeverCook/EditorUtilities/GenericAssetReviewTool/Art/M_GART_ListItemBG
graph {
 properties {materialDomain=MD_UI, pixelDepthOffsetMode=PDOM_Legacy, usedShadingModels=Default Lit}
 nodes [
  n1 = VectorParameter() {parameterName=Color, defaultValue=<rgba 1 0.0693 0 150>}
  n2 = ScalarParameter() {defaultValue=3, parameterName=HighlightMult}
  n3 = MaterialFunctionCall(Float2=TextureCoordinate()) {materialFunction=&BreakOutFloat2Components}
  n4 = ConstantBiasScale(None=@n3.R) {bias=-0.5, scale=2}
  n5 = Saturate(None=InverseLinearInterpolate(A=ScalarParameter() {defaultValue=0.65, parameterName=MaskMin}, Value=Abs(None=@n4)))
  n6 = InverseLinearInterpolate(Value=@n3.G) {constA=0.1}
  n7 = InverseLinearInterpolate(A=ScalarParameter() {defaultValue=0.7, parameterName=MaskVert}, Value=@n3.G)
  n8 = Power(Base=Max(A=Multiply(A=@n5, B=@n6), B=Saturate(None=@n7)), Exp=ScalarParameter() {defaultValue=1.5, parameterName=MaskExp})
  LinearInterpolate(
   A=@n1.RGB
   B=LinearInterpolate(A=@n1.RGB, B=Multiply(A=@n1.RGB, B=@n2), Alpha=Saturate(None=@n8))
   Alpha=Saturate(None=ScalarParameter() {parameterName=Selected})
  )
 ]
}
