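# Masked UI material in the GenericAssetReviewTool editor utility art, named for the report item combobox.
# It samples the Icon_DragEdge texture through CustomRotator, turning the icon a quarter turn.
# A Selected parameter blends world XY toward a copy pushed radially from the UV center, scaled by the gasfs parameter.
# No connections to material outputs are shown, and one material function call has no function assigned.
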Material /Game/NeverCook/EditorUtilities/GenericAssetReviewTool/Art/M_GART_ReportItemCombobox
graph {
 properties {blendMode=BLEND_Masked, materialDomain=MD_UI, pixelDepthOffsetMode=PDOM_Legacy, usedShadingModels=Default Lit}
 nodes [
  n1 = WorldPosition() {worldPositionShaderOffset=WPT_ExcludeAllShaderOffsets}
  n2 = Normalize(VectorInput=Subtract(A=TextureCoordinate()) {constB=0.5})
  n3 = ScalarParameter() {defaultValue=3, parameterName=gasfs}
  LinearInterpolate(A=@n1.XY, B=Add(A=@n1.XY, B=Multiply(A=@n2, B=@n3)), Alpha=ScalarParameter() {parameterName=Selected})
  MaterialFunctionCall()
  TextureSample(UVs=MaterialFunctionCall(UVs=TextureCoordinate(), "Rotation Angle (0-1)"=Constant() {r=0.25}) {materialFunction=&CustomRotator}.Rotated Values) {texture=/Game/Blueprints/Blutilities/ForTechArt/ArtAssets/Textures/Checkbox/Icon_DragEdge}
 ]
}
