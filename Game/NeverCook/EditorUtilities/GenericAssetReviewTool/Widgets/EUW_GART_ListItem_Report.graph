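# Editor utility list item that shows one report in the Generic Asset Review Tool (GART).
# SetReportData fills text, tooltip, color and severity overlay, then builds a dropdown to select or find assets, remove the item, or run report-exposed functions.
# Exposed functions can prompt with an object details dialog, then run on the loaded report asset through Python call_method.
# Dispatchers broadcast selection and removal; SetSelectedState drives the highlight.
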EditorUtilityWidgetBlueprint /Game/NeverCook/EditorUtilities/GenericAssetReviewTool/Widgets/EUW_GART_ListItem_Report
graphs [
 AddOptionsToComboBox {nodes [
   AddOptionsToComboBox() {is=title K2Node_FunctionEntry}
   > n1 = Sequence!(execute) -> (then_0, then_1, then_2:Exec)
   n2 = Get CurrentReportData() -> (CurrentReportData:Struct GART Report Output Data Structure)
   > n3 = Branch(Condition=IsValid(Object=Break Struct GART Report Output Data!(Struct_GART_ReportOutputData:"Struct GART Report Output Data Structure (by ref)"=@n2) -> (Object_39_#1:Object Reference) {is=title K2Node_BreakStruct}))
   > Set l_selectionString(l_selectionString:String=Select Object) -> (Output_Get:String)
   Set l_selectionString(execute=@n3.else, l_selectionString=Find Asset)
   n4 = Get EditorUtilityComboBoxString_124() -> (EditorUtilityComboBoxString_124:Editor Utility Combo Box String Object Reference)
   AddOption!(execute=@n1.then_1, self:"Combo Box (String) Object Reference"=@n4, Option:String=Get l_selectionString() -> (l_selectionString:String)) -> (then)
   > AddOption(self=@n4, Option=Find Report)
   > AddOption(self=@n4, Option=Remove)
   n5 = LoadAsset_Blocking(
    execute=@n1.then_2
    Asset=Break Struct GART Report Output Data!(Struct_GART_ReportOutputData=Get CurrentReportData()) -> (
     ReportData_46_#2:Struct GART Report Data Structure
     AssetData_36_#3:Asset Data Structure
     Object_39_#1
     Report_44_#4:Soft Object Reference
    ).Report_44_#4
   )
   > n6 = Cast To GART_Report_Base!(execute, Object=@n5) -> (then, CastFailed, AsGART Report Base:GART Report Base Object Reference) {is=title K2Node_DynamicCast}
   n7 = Get ExposedFunctions!(self:GART Report Base Object Reference=@n6) -> (ExposedFunctions:Array of GART Exposed Function Struct Structures)
   > Branch(Condition=Is Valid Index!(TargetArray:Array of GART Exposed Function Struct Structures=@n7, IndexToTest:int) -> (ReturnValue:bool))
   > AddOption(self=Get EditorUtilityComboBoxString_124(), Option=Get ComboBoxDividerString() -> (ComboBoxDividerString:String))
   > n8 = For Each Loop(Array:Array of GART Exposed Function Struct Structures=@n7) -> (Array Element:GART Exposed Function Struct Structure)
   n9 = Break GART Exposed Function Struct!(GART_ExposedFunction_Struct:"GART Exposed Function Struct Structure (by ref)"=@n8) -> (FunctionName_17_#5:String, FunctionDataObjectClass_16_#6:Object Class Reference) {is=title K2Node_BreakStruct}
   > Add!(
    execute
    TargetMap:Map of Strings to GART Exposed Function Struct Structures=Get CurrentExposedFunctions() -> (CurrentExposedFunctions:Map of Strings to GART Exposed Function Struct Structures)
    Key:"String (by ref)"=@n9
    Value:"GART Exposed Function Struct Structure (by ref)"=@n8
   ) -> (then) {is=title K2Node_CallFunction}
   > AddOption(self=Get EditorUtilityComboBoxString_124(), Option=@n9)
  ]}
 EventGraph {nodes [
   n1 = "On Selection Changed (EditorUtilityComboBoxString_124)"!() -> (OutputDelegate, then, SelectedItem:String, SelectionType:ESelectInfo Enum) {is=title K2Node_ComponentBoundEvent}
   > Branch(Condition="Not Equal Exactly (String)"(A=@n1.SelectedItem, B=Get ComboBoxDividerString()))
   > n2 = Sequence!(execute) -> (then_0, then_1)
   > n3 = Switch on ESelectInfo!(execute, Selection:ESelectInfo Enum=@n1.SelectionType) -> (OnKeyPress:Exec, OnNavigation:Exec, OnMouseClick:Exec, Direct:Exec) {is=title K2Node_SwitchEnum}
   n4 = Switch on String!(execute:Exec=[@n3.OnKeyPress, @n3.OnMouseClick], Selection:String=@n1.SelectedItem) -> (Default:Exec, Select Object:Exec, Find Asset:Exec, Find Report:Exec, Remove:Exec) {pinNames [Select Object, Find Asset, Find Report, Remove]}
   > TryCallExposedFunction!(execute, self, functionName:String=@n1.SelectedItem) -> (then)
   n5 = Break Struct GART Report Output Data!(Struct_GART_ReportOutputData=Get CurrentReportData()) -> (AssetData_36_#3, Report_44_#4)
   n6 = Break Asset Data(AssetData=@n5)
   SyncBrowserToObjects(execute=@n4.Find Asset, AssetPaths=Make Array("[0]":String="To String (Name)"(InName=@n6)) -> (Array:Array of Strings))
   n7 = "To String (SoftObjectReference)"!(SoftObjectReference:"Soft Object Reference (by ref)"=@n5.Report_44_#4) -> (ReturnValue:String)
   SyncBrowserToObjects(execute=@n4.Find Report, AssetPaths=Make Array("[0]"=Split(SourceString=@n7, InStr=., SearchDir=FromStart)))
   Call On Deleted!(execute:Exec=@n4.Remove, self:EUW GART List Item Report Object Reference, widget:EUW GART List Item Report Object Reference=Self-Reference()) -> (then:Exec) {is=title K2Node_CallDelegate}
   n8 = Acquire Editor Object Element Handle!(
    execute=@n4.Select Object
    Object:Object Reference=Break Struct GART Report Output Data!(Struct_GART_ReportOutputData=Get CurrentReportData()) -> (Object_39_#1:Object Reference)
    bAllowCreate:bool=true
   ) -> (then, ReturnValue:Script Typed Element Handle Structure)
   > n9 = GetSelectionSet(self=Get LevelEditorSubsystem())
   > SetSelection!(
    execute
    self:Typed Element Selection Set Object Reference=@n9
    InElementHandles:Array of Script Typed Element Handle Structures=Make Array("[0]":Script Typed Element Handle Structure=@n8) -> (Array:Array of Script Typed Element Handle Structures)
    InSelectionOptions:Typed Element Selection Options Structure=Make Typed Element Selection Options!(
     bAllowHidden:bool=true
     bAllowGroups:bool=True
     bAllowLegacyNotifications:bool=True
     bWarnIfLocked:bool=False
     bAllowSubRootSelection:bool=False
     ChildElementInclusionMethod:ETypedElementChildInclusionMethod Enum=Immediate
    ) -> (TypedElementSelectionOptions:Typed Element Selection Options Structure) {is=title K2Node_MakeStruct}
   ) -> (then, ReturnValue:bool)
   ClearSelection(execute=@n2.then_1, self:"Combo Box (String) Object Reference"=Get EditorUtilityComboBoxString_124())
  ]}
 OnDeleted {nodes [OnDeleted!() -> (then, widget:EUW GART List Item Report Object Reference) {is=title K2Node_FunctionEntry}]}
 OnMouseButtonDown {nodes [
   OnMouseButtonDown!() -> (then, MyGeometry:Geometry Structure, MouseEvent:"Pointer Event Structure (by ref)") {is=title K2Node_FunctionEntry}
   > n1 = Sequence()
   > Call On Selected!(execute:Exec, self:EUW GART List Item Report Object Reference, ReportWidget:EUW GART List Item Report Object Reference=Self-Reference()) -> (then:Exec) {is=title K2Node_CallDelegate}
   Return Node!(execute=@n1.then_1, ReturnValue:Event Reply Structure=Unhandled())
   Handled()
   SetSelectedState!(execute, self, selected:bool=true) -> (then)
   Return Node(ReturnValue=Unhandled())
  ]}
 OnPreviewMouseButtonDown {nodes [
   n1 = Sequence()
   > Call On Selected(ReportWidget=Self-Reference())
   Return Node(execute=@n1.then_1, ReturnValue=Handled())
   OnPreviewMouseButtonDown!() -> (then, MyGeometry:Geometry Structure, MouseEvent:"Pointer Event Structure (by ref)") {is=title K2Node_FunctionEntry}
   > Return Node(ReturnValue=Unhandled())
  ]}
 SetReportData {nodes [
   n1 = SetReportData!() -> (then, CurrentReportData:Struct GART Report Output Data Structure) {is=title K2Node_FunctionEntry}
   > Set CurrentReportData(CurrentReportData:Struct GART Report Output Data Structure=@n1) -> (Output_Get:Struct GART Report Output Data Structure)
   n2 = Break Struct GART Report Data!(Struct_GART_ReportData:"Struct GART Report Data Structure (by ref)"=Break Struct GART Report Output Data!(Struct_GART_ReportOutputData=Get CurrentReportData()) -> (ReportData_46_#2:Struct GART Report Data Structure)) -> (
    DisplayTitle_20_#7:String
    DisplayDescShort_29_#8:String
    DisplayDescLong_30_#9:String
    DisplayColorOverride_40_#10:Linear Color Structure
    Tags_25_#11:Array of Names
    Severity_46_#2:Enum_GART_ReportSeverity Enum=~
   ) {is=title K2Node_BreakStruct}
   > "SetText (Text)"(self=Get txtReportTitle() -> (txtReportTitle:Common Text Object Reference), InText="To Text (String)"(InString=@n2))
   > "SetText (Text)"(
    self=Get txtDesc() -> (txtDesc:Common Text Object Reference)
    InText="To Text (String)"(InString=@n2.DisplayDescShort_29_#8)
   )
   > SetToolTipText!(
    execute
    self:Widget Object Reference
    InToolTipText:"Text (by ref)"="To Text (String)"(InString=@n2.DisplayDescLong_30_#9)
   ) -> (then)
   > AddOptionsToComboBox() {is=title K2Node_CallFunction}
   > n3 = GetDynamicMaterial(self:Border Object Reference=Get borderMain() -> (borderMain:Border Object Reference))
   > SetVectorParameterValue(self=@n3, ParameterName=Color, Value=@n2.DisplayColorOverride_40_#10)
   > SetVisibility(
    self=Get imgBlocker() -> (imgBlocker:Image Object Reference)
    InVisibility=Select(
     Option 0=Collapsed
     Option 1=HitTestInvisible
     Index="Equal (Enum)"(A:Enum_GART_ReportSeverity Enum=@n2.Severity_46_#2, B:Enum_GART_ReportSeverity Enum=NewEnumerator3)
    )
   )
  ]}
 SetSelectedState {nodes [
   n1 = SetSelectedState!() -> (then, selected:bool=~) {is=title K2Node_FunctionEntry}
   > Set isSelected(isSelected:bool=@n1) -> (Output_Get:bool)
   > SetVisibility(
    self=[Get imgSelected() -> (imgSelected:Image Object Reference), Get imgSelectArrow() -> (imgSelectArrow:Image Object Reference)]
    InVisibility=Select(Option 0=Collapsed, Option 1=SelfHitTestInvisible, Index=@n1)
   )
   > n2 = GetDynamicMaterial(self=Get borderMain())
   > SetScalarParameterValue(self=@n2, ParameterName=Selected, Value="To Float (Boolean)"(InBool=Get isSelected() -> (isSelected:bool)))
   > SetVisibility(self=Get imgSelectArrow(), InVisibility=Select(Option 0=Collapsed, Option 1=SelfHitTestInvisible, Index=Get isSelected()))
  ]}
 TryCallExposedFunction {nodes [
   TryCallExposedFunction!() -> (then, functionName:String) {is=title K2Node_FunctionEntry}
   > n1 = Sequence!(execute) -> (then_0, then_1, then_2:Exec)
   > n2 = Find!(
    execute:Exec
    TargetMap:Map of Strings to GART Exposed Function Struct Structures=Get CurrentExposedFunctions()
    Key=Get functionName() -> (functionName:String)
   ) -> (then:Exec, Value:GART Exposed Function Struct Structure, ReturnValue)
   > Branch(Condition=@n2.ReturnValue)
   n3 = Break GART Exposed Function Struct(GART_ExposedFunction_Struct=@n2)
   > Branch(Condition=IsValidClass!(Class:Object Class Reference=@n3.FunctionDataObjectClass_16_#6) -> (ReturnValue:bool))
   > n4 = Construct Object from Class(Class=@n3.FunctionDataObjectClass_16_#6) -> (ReturnValue:Object Reference)
   > n5 = Set l_object(l_object:Object Reference=@n4) -> (Output_Get:Object Reference)
   > n6 = Show Object Dialog!(
    execute
    Title:"Text (by ref)"=Format Text(Format="To Text (String)"(InString=@n3))
    InOutObject:Object Reference=@n5
    Options:"Editor Dialog Library Object Details View Options Structure (by ref)"=Make Editor Dialog Library Object Details View Options!(
     bShowObjectName:bool=false
     bAllowSearch:bool=True
     bAllowResizing:bool=true
     MinWidth:int=0
     MinHeight:int=0
     ValueColumnWidthRatio:float=0.6
    ) -> (EditorDialogLibraryObjectDetailsViewOptions:Editor Dialog Library Object Details View Options Structure) {is=title K2Node_MakeStruct}
   ) -> (then, ReturnValue:bool)
   > n7 = Branch(Condition=@n6)
   Return Node!(execute=@n7.else)
   n8 = LoadAsset_Blocking(
    execute=@n1.then_1
    Asset=Break Struct GART Report Output Data!(Struct_GART_ReportOutputData=Get CurrentReportData()) -> (
     ReportData_46_#2
     AssetData_36_#3:Asset Data Structure
     Object_39_#1:Object Reference
     Report_44_#4:Soft Object Reference
    ).Report_44_#4
   )
   > n9 = Cast To GART_Report_Base(Object=@n8)
   > n10 = Set l_report(l_report:GART Report Base Object Reference=@n9) -> (Output_Get:GART Report Base Object Reference)
   > n11 = Is Valid(InputObject=@n10)
   Return Node(execute=@n11.Is Not Valid)
   Execute Python Script!(
    execute=@n1.then_2
    PythonScript="import unreal\r\n\r\n\r\ndef TryCallMethod():\r\n    ##cdo = unreal.get_default_object(in_report.generated_class())\r\n    cdo = in_report\r\n\r\n    if cdo is None:\r\n        unreal.log_error(\"Report tried to call function but no valid report asset found.\")\r\n        return\r\n\r\n    try:\r\n        if in_object is not None:\r\n            cdo.call_method(function_name, (report_data, in_object))\r\n        else:\r\n            cdo.call_method(function_name, (report_data,))\r\n    except:\r\n        unreal.log_error(\"Report tried to call function \" + function_name + \" but it did not exist.\")\r\n        return\r\n\r\nTryCallMethod()"
    functionName:String=Get functionName()
    in_object:Object Reference=Get l_object() -> (l_object:Object Reference)
    in_report:GART Report Base Object Reference=Get l_report() -> (l_report:GART Report Base Object Reference)
    report_data:Struct GART Report Output Data Structure=Get CurrentReportData()
   ) -> (then, ReturnValue) {inputs [functionName,in_object,in_report,report_data]}
  ]}
 onSelected {nodes [onSelected!() -> (then, ReportWidget:EUW GART List Item Report Object Reference) {is=title K2Node_FunctionEntry}]}
]
variables [
 ComboBoxDividerString {type=string, default="=---------="}
 CurrentExposedFunctions {type="map<GART_ExposedFunction_Struct>", default {}}
 CurrentReportData {
  type=Struct_GART_ReportOutputData
  default {
   object=null
   report=null
   assetData {}
   reportData {displayDescLong="", displayDescShort="", displayTitle="", severity=Note, displayColorOverride=<rgba 0 0 0 0>, tags []}
  }
 }
 isSelected {type=boolean}
 OnDeleted {default=()}
 onSelected {default=()}
]
widgets {
 info {rootWidgetClass=&Overlay, widgetCount=15}
 widgets [
  Overlay_0 {widgetClassPath=&Overlay, widget=Overlay_0}
  Overlay_2 {widgetClassPath=&Overlay, parent=Overlay_0, widget=Overlay_2, slot {class=OverlaySlot, properties {padding=<ltrb 5 0 0 0>}}}
  borderMain {
   widgetClassPath=&Border
   bIsVariable=true
   parent=Overlay_2
   widget=borderMain
   properties {
    background {
     bIsDynamicallyLoaded=false
     drawAs=Image
     imageType=FullColor
     mirroring=NoMirror
     resourceName=None
     resourceObject=/Game/NeverCook/EditorUtilities/GenericAssetReviewTool/Art/M_GART_ListItemBG
     tiling=NoTile
     imageSize=<32 32>
     margin=<ltrb 0 0 0 0>
     outlineSettings {
      bUseBrushTransparency=false
      roundingType=HalfHeightRadius
      width=0
      color {colorUseRule=UseColor_Specified, specifiedColor=<rgba 0 0 0 0>}
      cornerRadii=<0 0 0 0>
     }
     tintColor {colorUseRule=UseColor_Specified, specifiedColor=<rgba 1 1 1 1>}
     uVRegion {bIsValid=false, max=<0 0>, min=<0 0>}
    }
    padding=<ltrb 0 0 0 0>
   }
   slot {class=OverlaySlot}
  }
  SizeBox_0 {
   widgetClassPath=&SizeBox
   parent=borderMain
   widget=SizeBox_0
   properties {bOverride_MinDesiredHeight=true, bOverride_WidthOverride=true, heightOverride=54.3448, minDesiredHeight=64, widthOverride=111.3304}
   slot {class=BorderSlot, properties {padding=<ltrb 0 0 0 0>}}
  }
  VerticalBox_0 {
   widgetClassPath=&VerticalBox
   parent=SizeBox_0
   widget=VerticalBox_0
   slot {class=SizeBoxSlot, properties {padding=<ltrb 0 1 0 0>}}
  }
  HorizontalBox_0 {widgetClassPath=&HorizontalBox, parent=VerticalBox_0, widget=HorizontalBox_0, slot {class=VerticalBoxSlot}}
  txtReportTitle {
   widgetClassPath=&CommonTextBlock
   bIsVariable=true
   parent=HorizontalBox_0
   widget=txtReportTitle
   properties {
    text=Report Title
    font {
     bForceMonospaced=false
     bMaterialIsStencil=false
     fontMaterial=null
     fontObject=&NotoSans
     letterSpacing=0
     monospacedWidth=1
     size=9
     skewAmount=0
     typefaceFontName=NotoSans-Regular
     outlineSettings {
      bApplyOutlineToDropShadows=false
      bMiteredCorners=false
      bSeparateFillAlpha=false
      outlineMaterial=null
      outlineSize=0
      outlineColor=<rgba 0 0 0 1>
     }
    }
    shadowOffset=<0 0>
   }
   slot {class=HorizontalBoxSlot, properties {padding=<ltrb 1 0 0 0>, size {sizeRule=Fill, value=1}}}
  }
  Overlay_1 {widgetClassPath=&Overlay, parent=HorizontalBox_0, widget=Overlay_1, slot {class=HorizontalBoxSlot}}
  SizeBox_1 {
   widgetClassPath=&SizeBox
   parent=Overlay_1
   widget=SizeBox_1
   properties {bOverride_HeightOverride=true, bOverride_WidthOverride=true, heightOverride=16, widthOverride=16}
   slot {class=OverlaySlot}
  }
  EditorUtilityComboBoxString_124 {
   widgetClassPath=/Script/Blutility.EditorUtilityComboBoxString
   bIsVariable=true
   parent=SizeBox_1
   widget=EditorUtilityComboBoxString_124
   properties {
    clipping=ClipToBounds
    font {
     bForceMonospaced=false
     bMaterialIsStencil=false
     fontMaterial=null
     fontObject=&Roboto
     letterSpacing=0
     monospacedWidth=1
     size=10.5
     skewAmount=0
     typefaceFontName=Bold
     outlineSettings {
      bApplyOutlineToDropShadows=false
      bMiteredCorners=false
      bSeparateFillAlpha=false
      outlineMaterial=null
      outlineSize=0
      outlineColor=<rgba 0 0 0 1>
     }
    }
    foregroundColor {colorUseRule=UseColor_Specified, specifiedColor=<rgba 0.5271 0.5271 0.5271 1>}
    itemStyle {
     bUseParentRowBrush=false
     activeBrush {
      bIsDynamicallyLoaded=false
      drawAs=Image
      imageType=NoImage
      mirroring=NoMirror
      resourceName=None
      resourceObject=null
      tiling=NoTile
      imageSize=<0 0>
      margin=<ltrb 0 0 0 0>
      outlineSettings {
       bUseBrushTransparency=false
       roundingType=HalfHeightRadius
       width=0
       color {colorUseRule=UseColor_Specified, specifiedColor=<rgba 0 0 0 0>}
       cornerRadii=<0 0 0 0>
      }
      tintColor {colorUseRule=UseColor_Specified, specifiedColor=<rgba 0 0.162 0.7454 1>}
      uVRegion {bIsValid=false, max=<0 0>, min=<0 0>}
     }
     activeHighlightedBrush {
      bIsDynamicallyLoaded=false
      drawAs=Image
      imageType=NoImage
      mirroring=NoMirror
      resourceName=None
      resourceObject=null
      tiling=NoTile
      imageSize=<0 0>
      margin=<ltrb 0 0 0 0>
      outlineSettings {
       bUseBrushTransparency=false
       roundingType=HalfHeightRadius
       width=0
       color {colorUseRule=UseColor_Specified, specifiedColor=<rgba 0 0 0 0>}
       cornerRadii=<0 0 0 0>
      }
      tintColor {colorUseRule=UseColor_Specified, specifiedColor=<rgba 0.0252 0.0319 0.0423 1>}
      uVRegion {bIsValid=false, max=<0 0>, min=<0 0>}
     }
     activeHoveredBrush {
      bIsDynamicallyLoaded=false
      drawAs=Image
      imageType=NoImage
      mirroring=NoMirror
      resourceName=None
      resourceObject=null
      tiling=NoTile
      imageSize=<0 0>
      margin=<ltrb 0 0 0 0>
      outlineSettings {
       bUseBrushTransparency=false
       roundingType=HalfHeightRadius
       width=0
       color {colorUseRule=UseColor_Specified, specifiedColor=<rgba 0 0 0 0>}
       cornerRadii=<0 0 0 0>
      }
      tintColor {colorUseRule=UseColor_Specified, specifiedColor=<rgba 0 0.162 0.7454 1>}
      uVRegion {bIsValid=false, max=<0 0>, min=<0 0>}
     }
     dropIndicator_Above {
      bIsDynamicallyLoaded=false
      drawAs=Box
      imageType=FullColor
      mirroring=NoMirror
      resourceName=../../../Engine/Content/Slate/Common/DropZoneIndicator_Above.png
      resourceObject=null
      tiling=NoTile
      imageSize=<0 0>
      margin=<ltrb 0.625 0.625 0 0>
      outlineSettings {
       bUseBrushTransparency=false
       roundingType=HalfHeightRadius
       width=0
       color {colorUseRule=UseColor_Specified, specifiedColor=<rgba 0 0 0 0>}
       cornerRadii=<0 0 0 0>
      }
      tintColor {colorUseRule=UseColor_Specified, specifiedColor=<rgba 0.0091 0.3515 0.9647 1>}
      uVRegion {bIsValid=false, max=<0 0>, min=<0 0>}
     }
     dropIndicator_Below {
      bIsDynamicallyLoaded=false
      drawAs=Box
      imageType=FullColor
      mirroring=NoMirror
      resourceName=../../../Engine/Content/Slate/Common/DropZoneIndicator_Below.png
      resourceObject=null
      tiling=NoTile
      imageSize=<0 0>
      margin=<ltrb 0.625 0 0 0.625>
      outlineSettings {
       bUseBrushTransparency=false
       roundingType=HalfHeightRadius
       width=0
       color {colorUseRule=UseColor_Specified, specifiedColor=<rgba 0 0 0 0>}
       cornerRadii=<0 0 0 0>
      }
      tintColor {colorUseRule=UseColor_Specified, specifiedColor=<rgba 0.0091 0.3515 0.9647 1>}
      uVRegion {bIsValid=false, max=<0 0>, min=<0 0>}
     }
     dropIndicator_Onto {
      bIsDynamicallyLoaded=false
      drawAs=Box
      imageType=FullColor
      mirroring=NoMirror
      resourceName=../../../Engine/Content/Slate/Common/DropZoneIndicator_Onto.png
      resourceObject=null
      tiling=NoTile
      imageSize=<0 0>
      margin=<ltrb 0.25 0.25 0.25 0.25>
      outlineSettings {
       bUseBrushTransparency=false
       roundingType=HalfHeightRadius
       width=0
       color {colorUseRule=UseColor_Specified, specifiedColor=<rgba 0 0 0 0>}
       cornerRadii=<0 0 0 0>
      }
      tintColor {colorUseRule=UseColor_Specified, specifiedColor=<rgba 0.0091 0.3515 0.9647 1>}
      uVRegion {bIsValid=false, max=<0 0>, min=<0 0>}
     }
     evenRowBackgroundBrush {
      bIsDynamicallyLoaded=false
      drawAs=Image
      imageType=NoImage
      mirroring=NoMirror
      resourceName=None
      resourceObject=null
      tiling=NoTile
      imageSize=<0 0>
      margin=<ltrb 0 0 0 0>
      outlineSettings {
       bUseBrushTransparency=false
       roundingType=HalfHeightRadius
       width=0
       color {colorUseRule=UseColor_Specified, specifiedColor=<rgba 0 0 0 0>}
       cornerRadii=<0 0 0 0>
      }
      tintColor {colorUseRule=UseColor_Specified, specifiedColor=<rgba 0.0103 0.0103 0.0103 1>}
      uVRegion {bIsValid=false, max=<0 0>, min=<0 0>}
     }
     evenRowBackgroundHoveredBrush {
      bIsDynamicallyLoaded=false
      drawAs=Image
      imageType=NoImage
      mirroring=NoMirror
      resourceName=None
      resourceObject=null
      tiling=NoTile
      imageSize=<0 0>
      margin=<ltrb 0 0 0 0>
      outlineSettings {
       bUseBrushTransparency=false
       roundingType=HalfHeightRadius
       width=0
       color {colorUseRule=UseColor_Specified, specifiedColor=<rgba 0 0 0 0>}
       cornerRadii=<0 0 0 0>
      }
      tintColor {colorUseRule=UseColor_Specified, specifiedColor=<rgba 0.0176 0.0176 0.0176 1>}
      uVRegion {bIsValid=false, max=<0 0>, min=<0 0>}
     }
     inactiveBrush {
      bIsDynamicallyLoaded=false
      drawAs=Image
      imageType=NoImage
      mirroring=NoMirror
      resourceName=None
      resourceObject=null
      tiling=NoTile
      imageSize=<0 0>
      margin=<ltrb 0 0 0 0>
      outlineSettings {
       bUseBrushTransparency=false
       roundingType=HalfHeightRadius
       width=0
       color {colorUseRule=UseColor_Specified, specifiedColor=<rgba 0 0 0 0>}
       cornerRadii=<0 0 0 0>
      }
      tintColor {colorUseRule=UseColor_Specified, specifiedColor=<rgba 0.0513 0.0953 0.159 1>}
      uVRegion {bIsValid=false, max=<0 0>, min=<0 0>}
     }
     inactiveHighlightedBrush {
      bIsDynamicallyLoaded=false
      drawAs=Image
      imageType=NoImage
      mirroring=NoMirror
      resourceName=None
      resourceObject=null
      tiling=NoTile
      imageSize=<0 0>
      margin=<ltrb 0 0 0 0>
      outlineSettings {
       bUseBrushTransparency=false
       roundingType=HalfHeightRadius
       width=0
       color {colorUseRule=UseColor_Specified, specifiedColor=<rgba 0 0 0 0>}
       cornerRadii=<0 0 0 0>
      }
      tintColor {colorUseRule=UseColor_Specified, specifiedColor=<rgba 0.0252 0.0319 0.0423 1>}
      uVRegion {bIsValid=false, max=<0 0>, min=<0 0>}
     }
     inactiveHoveredBrush {
      bIsDynamicallyLoaded=false
      drawAs=Image
      imageType=NoImage
      mirroring=NoMirror
      resourceName=None
      resourceObject=null
      tiling=NoTile
      imageSize=<0 0>
      margin=<ltrb 0 0 0 0>
      outlineSettings {
       bUseBrushTransparency=false
       roundingType=HalfHeightRadius
       width=0
       color {colorUseRule=UseColor_Specified, specifiedColor=<rgba 0 0 0 0>}
       cornerRadii=<0 0 0 0>
      }
      tintColor {colorUseRule=UseColor_Specified, specifiedColor=<rgba 0.0513 0.0953 0.159 1>}
      uVRegion {bIsValid=false, max=<0 0>, min=<0 0>}
     }
     oddRowBackgroundBrush {
      bIsDynamicallyLoaded=false
      drawAs=Image
      imageType=NoImage
      mirroring=NoMirror
      resourceName=None
      resourceObject=null
      tiling=NoTile
      imageSize=<0 0>
      margin=<ltrb 0 0 0 0>
      outlineSettings {
       bUseBrushTransparency=false
       roundingType=HalfHeightRadius
       width=0
       color {colorUseRule=UseColor_Specified, specifiedColor=<rgba 0 0 0 0>}
       cornerRadii=<0 0 0 0>
      }
      tintColor {colorUseRule=UseColor_Specified, specifiedColor=<rgba 0.0103 0.0103 0.0103 1>}
      uVRegion {bIsValid=false, max=<0 0>, min=<0 0>}
     }
     oddRowBackgroundHoveredBrush {
      bIsDynamicallyLoaded=false
      drawAs=Image
      imageType=NoImage
      mirroring=NoMirror
      resourceName=None
      resourceObject=null
      tiling=NoTile
      imageSize=<0 0>
      margin=<ltrb 0 0 0 0>
      outlineSettings {
       bUseBrushTransparency=false
       roundingType=HalfHeightRadius
       width=0
       color {colorUseRule=UseColor_Specified, specifiedColor=<rgba 0 0 0 0>}
       cornerRadii=<0 0 0 0>
      }
      tintColor {colorUseRule=UseColor_Specified, specifiedColor=<rgba 0.0176 0.0176 0.0176 1>}
      uVRegion {bIsValid=false, max=<0 0>, min=<0 0>}
     }
     parentRowBackgroundBrush {
      bIsDynamicallyLoaded=false
      drawAs=Image
      imageType=NoImage
      mirroring=NoMirror
      resourceName=None
      resourceObject=null
      tiling=NoTile
      imageSize=<32 32>
      margin=<ltrb 0 0 0 0>
      outlineSettings {
       bUseBrushTransparency=false
       roundingType=HalfHeightRadius
       width=0
       color {colorUseRule=UseColor_Specified, specifiedColor=<rgba 0 0 0 0>}
       cornerRadii=<0 0 0 0>
      }
      tintColor {colorUseRule=UseColor_Specified, specifiedColor=<rgba 1 1 1 1>}
      uVRegion {bIsValid=false, max=<0 0>, min=<0 0>}
     }
     parentRowBackgroundHoveredBrush {
      bIsDynamicallyLoaded=false
      drawAs=Image
      imageType=NoImage
      mirroring=NoMirror
      resourceName=None
      resourceObject=null
      tiling=NoTile
      imageSize=<32 32>
      margin=<ltrb 0 0 0 0>
      outlineSettings {
       bUseBrushTransparency=false
       roundingType=HalfHeightRadius
       width=0
       color {colorUseRule=UseColor_Specified, specifiedColor=<rgba 0 0 0 0>}
       cornerRadii=<0 0 0 0>
      }
      tintColor {colorUseRule=UseColor_Specified, specifiedColor=<rgba 1 1 1 1>}
      uVRegion {bIsValid=false, max=<0 0>, min=<0 0>}
     }
     selectedTextColor {colorUseRule=UseColor_Specified, specifiedColor=<rgba 1 1 1 1>}
     selectorFocusedBrush {
      bIsDynamicallyLoaded=false
      drawAs=Border
      imageType=FullColor
      mirroring=NoMirror
      resourceName=../../../Engine/Content/Slate/Common/Selector.png
      resourceObject=null
      tiling=Both
      imageSize=<0 0>
      margin=<ltrb 0.25 0.25 0.25 0.25>
      outlineSettings {
       bUseBrushTransparency=false
       roundingType=HalfHeightRadius
       width=0
       color {colorUseRule=UseColor_Specified, specifiedColor=<rgba 0 0 0 0>}
       cornerRadii=<0 0 0 0>
      }
      tintColor {colorUseRule=UseColor_Specified, specifiedColor=<rgba 0.701 0.225 0.003 1>}
      uVRegion {bIsValid=false, max=<0 0>, min=<0 0>}
     }
     textColor {colorUseRule=UseColor_Specified, specifiedColor=<rgba 0.5271 0.5271 0.5271 1>}
    }
    widgetStyle {
     comboButtonStyle {
      downArrowAlign=VAlign_Center
      buttonStyle {
       clickedSlateSound {resourceObject=null}
       disabled {
        bIsDynamicallyLoaded=false
        drawAs=RoundedBox
        imageType=FullColor
        mirroring=NoMirror
        resourceName=None
        resourceObject=/Game/NeverCook/EditorUtilities/GenericAssetReviewTool/Art/M_GART_ReportItemCombobox
        tiling=NoTile
        imageSize=<32 32>
        margin=<ltrb 0 0 0 0>
        outlineSettings {
         bUseBrushTransparency=false
         roundingType=FixedRadius
         width=1
         color {colorUseRule=UseColor_ColorTable, specifiedColor=<rgba 0 0 0 0>}
         cornerRadii=<4 4 4 4>
        }
        tintColor {colorUseRule=UseColor_Specified, specifiedColor=<rgba 0.5 0.5 0.5 0.5>}
        uVRegion {bIsValid=false, max=<0 0>, min=<0 0>}
       }
       disabledForeground {colorUseRule=UseColor_ColorTable, specifiedColor=<rgba 0 0 0 0>}
       hovered {
        bIsDynamicallyLoaded=false
        drawAs=RoundedBox
        imageType=FullColor
        mirroring=NoMirror
        resourceName=None
        resourceObject=&M_GART_ReportItemCombobox
        tiling=NoTile
        imageSize=<32 32>
        margin=<ltrb 0 0 0 0>
        outlineSettings {
         bUseBrushTransparency=false
         roundingType=FixedRadius
         width=1
         color {colorUseRule=UseColor_ColorTable, specifiedColor=<rgba 0 0 0 0>}
         cornerRadii=<4 4 4 4>
        }
        tintColor {colorUseRule=UseColor_Specified, specifiedColor=<rgba 1 1 1 0.5>}
        uVRegion {bIsValid=false, max=<0 0>, min=<0 0>}
       }
       hoveredForeground {colorUseRule=UseColor_ColorTable, specifiedColor=<rgba 0 0 0 0>}
       hoveredSlateSound {resourceObject=null}
       normal {
        bIsDynamicallyLoaded=false
        drawAs=RoundedBox
        imageType=FullColor
        mirroring=NoMirror
        resourceName=None
        resourceObject=&M_GART_ReportItemCombobox
        tiling=NoTile
        imageSize=<32 32>
        margin=<ltrb 0 0 0 0>
        outlineSettings {
         bUseBrushTransparency=false
         roundingType=FixedRadius
         width=1
         color {colorUseRule=UseColor_ColorTable, specifiedColor=<rgba 0 0 0 0>}
         cornerRadii=<4 4 4 4>
        }
        tintColor {colorUseRule=UseColor_Specified, specifiedColor=<rgba 0.5 0.5 0.5 0.5>}
        uVRegion {bIsValid=false, max=<0 0>, min=<0 0>}
       }
       normalForeground {colorUseRule=UseColor_ColorTable, specifiedColor=<rgba 0 0 0 0>}
       normalPadding=<ltrb 8 1 8 1>
       pressed {
        bIsDynamicallyLoaded=false
        drawAs=RoundedBox
        imageType=FullColor
        mirroring=NoMirror
        resourceName=None
        resourceObject=&M_GART_ReportItemCombobox
        tiling=NoTile
        imageSize=<32 32>
        margin=<ltrb 0 0 0 0>
        outlineSettings {
         bUseBrushTransparency=false
         roundingType=FixedRadius
         width=1
         color {colorUseRule=UseColor_ColorTable, specifiedColor=<rgba 0 0 0 0>}
         cornerRadii=<4 4 4 4>
        }
        tintColor {colorUseRule=UseColor_Specified, specifiedColor=<rgba 0.5 0.5 0.5 0.5>}
        uVRegion {bIsValid=false, max=<0 0>, min=<0 0>}
       }
       pressedForeground {colorUseRule=UseColor_ColorTable, specifiedColor=<rgba 0 0 0 0>}
       pressedPadding=<ltrb 8 2 8 0>
       pressedSlateSound {resourceObject=null}
      }
      contentPadding=<ltrb 0 0 0 0>
      downArrowImage {
       bIsDynamicallyLoaded=false
       drawAs=Image
       imageType=Vector
       mirroring=NoMirror
       resourceName=../../../Engine/Content/Slate/Starship/CoreWidgets/ComboBox/wide-chevron-down.svg
       resourceObject=null
       tiling=NoTile
       imageSize=<16 16>
       margin=<ltrb 0 0 0 0>
       outlineSettings {
        bUseBrushTransparency=false
        roundingType=HalfHeightRadius
        width=0
        color {colorUseRule=UseColor_Specified, specifiedColor=<rgba 0 0 0 0>}
        cornerRadii=<0 0 0 0>
       }
       tintColor {colorUseRule=UseColor_Specified, specifiedColor=<rgba 1 1 1 0>}
       uVRegion {bIsValid=false, max=<0 0>, min=<0 0>}
      }
      downArrowPadding=<ltrb 0 0 0 0>
      menuBorderBrush {
       bIsDynamicallyLoaded=false
       drawAs=RoundedBox
       imageType=NoImage
       mirroring=NoMirror
       resourceName=None
       resourceObject=null
       tiling=NoTile
       imageSize=<0 0>
       margin=<ltrb 0 0 0 0>
       outlineSettings {
        bUseBrushTransparency=false
        roundingType=FixedRadius
        width=1
        color {colorUseRule=UseColor_ColorTable, specifiedColor=<rgba 0 0 0 0>}
        cornerRadii=<0 0 0 0>
       }
       tintColor {colorUseRule=UseColor_Specified, specifiedColor=<rgba 0.0395 0.0395 0.0395 0>}
       uVRegion {bIsValid=false, max=<0 0>, min=<0 0>}
      }
      menuBorderPadding=<ltrb 0 0 0 0>
      shadowColorAndOpacity=<rgba 0 0 0 0>
      shadowOffset=<0 0>
     }
     contentPadding=<ltrb 0 0 0 0>
     menuRowPadding=<ltrb 8 3 8 3>
     pressedSlateSound {resourceObject=null}
     selectionChangeSlateSound {resourceObject=null}
    }
   }
   slot {class=SizeBoxSlot}
  }
  Image_146 {
   widgetClassPath=&Image
   bIsVariable=true
   parent=VerticalBox_0
   widget=Image_146
   properties {
    brush {
     bIsDynamicallyLoaded=false
     drawAs=Image
     imageType=NoImage
     mirroring=NoMirror
     resourceName=None
     resourceObject=null
     tiling=NoTile
     imageSize=<1 1>
     margin=<ltrb 0 0 0 0>
     outlineSettings {
      bUseBrushTransparency=false
      roundingType=HalfHeightRadius
      width=0
      color {colorUseRule=UseColor_Specified, specifiedColor=<rgba 0 0 0 0>}
      cornerRadii=<0 0 0 0>
     }
     tintColor {colorUseRule=UseColor_Specified, specifiedColor=<rgba 1 1 1 1>}
     uVRegion {bIsValid=false, max=<0 0>, min=<0 0>}
    }
    colorAndOpacity=<rgba 1 1 1 0.35>
   }
   slot {class=VerticalBoxSlot}
  }
  txtDesc {
   widgetClassPath=&CommonTextBlock
   bIsVariable=true
   parent=VerticalBox_0
   widget=txtDesc
   properties {
    autoWrapText=true
    justification=Center
    text=Longer Description might go here
    wrappingPolicy=AllowPerCharacterWrapping
    font {
     bForceMonospaced=false
     bMaterialIsStencil=false
     fontMaterial=null
     fontObject=&NotoSans
     letterSpacing=0
     monospacedWidth=1
     size=9
     skewAmount=0
     typefaceFontName=NotoSans-Regular
     outlineSettings {
      bApplyOutlineToDropShadows=false
      bMiteredCorners=false
      bSeparateFillAlpha=false
      outlineMaterial=null
      outlineSize=0
      outlineColor=<rgba 0 0 0 1>
     }
    }
    shadowOffset=<0 0>
   }
   slot {class=VerticalBoxSlot, properties {verticalAlignment=VAlign_Center, padding=<ltrb 1 1 1 1>, size {sizeRule=Fill, value=1}}}
  }
  imgSelected {
   widgetClassPath=&Image
   bIsVariable=true
   parent=VerticalBox_0
   widget=imgSelected
   properties {
    visibility=Collapsed
    brush {
     bIsDynamicallyLoaded=false
     drawAs=Image
     imageType=NoImage
     mirroring=NoMirror
     resourceName=None
     resourceObject=null
     tiling=NoTile
     imageSize=<32 2>
     margin=<ltrb 0 0 0 0>
     outlineSettings {
      bUseBrushTransparency=false
      roundingType=HalfHeightRadius
      width=0
      color {colorUseRule=UseColor_Specified, specifiedColor=<rgba 0 0 0 0>}
      cornerRadii=<0 0 0 0>
     }
     tintColor {colorUseRule=UseColor_Specified, specifiedColor=<rgba 1 1 1 1>}
     uVRegion {bIsValid=false, max=<0 0>, min=<0 0>}
    }
    colorAndOpacity=<rgba 1 0.2457 0 1>
   }
   slot {class=VerticalBoxSlot}
  }
  imgBlocker {
   widgetClassPath=&Image
   bIsVariable=true
   parent=Overlay_2
   widget=imgBlocker
   properties {
    clipping=ClipToBoundsWithoutIntersecting
    visibility=Collapsed
    brush {
     bIsDynamicallyLoaded=false
     drawAs=Box
     imageType=FullColor
     mirroring=NoMirror
     resourceName=None
     resourceObject=/Game/NeverCook/EditorUtilities/GenericAssetReviewTool/Art/M_GART_BlockerBG
     tiling=NoTile
     imageSize=<32 32>
     margin=<ltrb 0.26 0.26 0.74 0.74>
     outlineSettings {
      bUseBrushTransparency=false
      roundingType=HalfHeightRadius
      width=0
      color {colorUseRule=UseColor_Specified, specifiedColor=<rgba 0 0 0 0>}
      cornerRadii=<0 0 0 0>
     }
     tintColor {colorUseRule=UseColor_Specified, specifiedColor=<rgba 1 1 1 1>}
     uVRegion {bIsValid=false, max=<0 0>, min=<0 0>}
    }
    renderTransform {angle=0, scale=<1.01 1.01>, shear=<0 0>, translation=<0 0>}
   }
   slot {class=OverlaySlot, properties {horizontalAlignment=HAlign_Fill, verticalAlignment=VAlign_Fill}}
  }
  imgSelectArrow {
   widgetClassPath=&Image
   bIsVariable=true
   parent=Overlay_2
   widget=imgSelectArrow
   properties {
    clipping=ClipToBoundsWithoutIntersecting
    visibility=Collapsed
    brush {
     bIsDynamicallyLoaded=false
     drawAs=Image
     imageType=FullColor
     mirroring=NoMirror
     resourceName=None
     resourceObject=/Game/NeverCook/EditorUtilities/GenericAssetReviewTool/Art/M_GART_DownArrowSelect
     tiling=NoTile
     imageSize=<32 9>
     margin=<ltrb 0 0 0 0>
     outlineSettings {
      bUseBrushTransparency=false
      roundingType=HalfHeightRadius
      width=0
      color {colorUseRule=UseColor_Specified, specifiedColor=<rgba 0 0 0 0>}
      cornerRadii=<0 0 0 0>
     }
     tintColor {colorUseRule=UseColor_Specified, specifiedColor=<rgba 1 0.2963 0 1>}
     uVRegion {bIsValid=false, max=<0 0>, min=<0 0>}
    }
    renderTransform {angle=0, scale=<1 1>, shear=<0 0>, translation=<0 8.9>}
   }
   slot {class=OverlaySlot, properties {horizontalAlignment=HAlign_Fill, verticalAlignment=VAlign_Bottom}}
  }
 ]
}
