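# Editor utility list-entry widget for the Generic Asset Review Tool, showing one asset or object and its reports.
# When its list data is set, it fills the name and spawns a child report widget per entry in a horizontal scroll box.
# The selected report is written back to the list data and shows its long description when the row is selected.
# Buttons open, locate (Content Browser or level viewport) or delete the entry.
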EditorUtilityWidgetBlueprint /Game/NeverCook/EditorUtilities/GenericAssetReviewTool/Widgets/EUW_GART_ListItem_Main
graphs [
 AddReport {nodes [
   n1 = AddReport!() -> (then, ReportData:"Struct GART Report Output Data Structure (by ref)") {is=title K2Node_FunctionEntry}
   n2 = Break Struct GART Report Output Data!(Struct_GART_ReportOutputData:"Struct GART Report Output Data Structure (by ref)"=@n1) -> (
    ReportData_46_#1:Struct GART Report Data Structure
    AssetData_36_#2:Asset Data Structure
    Object_39_#3:Object Reference
    Report_44_#4:Soft Object Reference
   ) {is=title K2Node_BreakStruct}
   > PrintString(InString=Append(
     A="adding report "
     B=Break Struct GART Report Data!(Struct_GART_ReportData:"Struct GART Report Data Structure (by ref)"=@n2) -> (
      DisplayTitle_20_#5:String
      DisplayDescShort_29_#6:String
      DisplayDescLong_30_#7:String
      DisplayColorOverride_40_#8:Linear Color Structure
      Tags_25_#9:Array of Names
      Severity_46_#1:Enum_GART_ReportSeverity Enum=~
     ) {is=title K2Node_BreakStruct}
    ))
   > n3 = Create Widget!(
    execute:Exec
    WorldContextObject:Object Reference
    Class:User Widget Class Reference=&/EUW_GART_ListItem_Report.EUW_GART_ListItem_Report_C
    OwningPlayer:Player Controller Object Reference
   ) -> (then:Exec, ReturnValue:EUW GART List Item Report Object Reference)
   > SetReportData!(execute, self:EUW GART List Item Report Object Reference=@n3, CurrentReportData:Struct GART Report Output Data Structure=@n1) -> (then)
   > AddChild(self=getReportPanel!() -> (panel:Panel Widget Object Reference), Content=@n3)
   > Bind Event to On Deleted(self:EUW GART List Item Report Object Reference=@n3, Delegate=Create Event()) {is=title K2Node_AddDelegate}
   > Bind Event to On Selected(self:EUW GART List Item Report Object Reference=@n3, Delegate=Create Event()) {is=title K2Node_AddDelegate}
  ]}
 ClearReports {nodes [
   ClearReports() {is=title K2Node_FunctionEntry}
   > ClearChildren(self=getReportPanel())
  ]}
 DeleteSelf {nodes [
   DeleteSelf() {is=title K2Node_FunctionEntry}
   > n1 = Cast To ListView!(execute, Object=GetOwningListView!(UserListEntry:User List Entry Interface) -> (ReturnValue:List View Base Object Reference)) -> (then, CastFailed, AsList View:List View Object Reference) {is=title K2Node_DynamicCast}
   > RemoveItem!(
    execute
    self:List View Object Reference=@n1
    Item:Object Reference=GetListItemObject!(UserObjectListEntry:User Object List Entry Interface) -> (ReturnValue:Object Reference)
   ) -> (then)
  ]}
 EventGraph {nodes [
   "On Clicked (btnEditAsset)"() {is=title K2Node_ComponentBoundEvent}
   n1 = Break Struct GART Asset Report List!(Struct_GART_AssetReportList:"Struct GART Asset Report List Structure (by ref)"=Get CurrentReportData() -> (CurrentReportData:Struct GART Asset Report List Structure)) -> (
    ReportDatas_11_#10:Array of Struct GART Report Output Data Structures
    AssetData_7_#11:Asset Data Structure
    Object_10_#12:Object Reference
   ) {is=title K2Node_BreakStruct}
   > OpenEditorForAssets!(
    execute
    self:Asset Editor Subsystem Object Reference=Get AssetEditorSubsystem()
    Assets:Array of Object References=Make Array("[0]":Object Reference=GetAsset(InAssetData=@n1.AssetData_7_#11)) -> (Array:Array of Object References)
    OpenedMethod:EAssetTypeActivationOpenedMethod Enum=Edit
   ) -> (then, ReturnValue:bool)
   "On Clicked (btnGoToAsset)"() {is=title K2Node_ComponentBoundEvent}
   n2 = Break Struct GART Asset Report List(Struct_GART_AssetReportList=Get CurrentReportData())
   > n3 = Is Valid(InputObject=@n2.Object_10_#12)
   n4 = Break Asset Data(AssetData=@n2.AssetData_7_#11)
   SyncBrowserToObjects(execute=@n3.Is Not Valid, AssetPaths=Make Array("[0]":String="To String (Name)"(InName=@n4)) -> (Array:Array of Strings))
   n5 = Cast To Actor!(execute=@n3.Is Valid, Object=@n2.Object_10_#12) -> (then, CastFailed, AsActor:Actor Object Reference) {is=title K2Node_DynamicCast}
   n6 = Make Array("[0]":Actor Object Reference=@n5) -> (Array:Array of Actor Object References)
   > Focus Actors in Viewport!(" ":Exec, actors:Array of Actor Object References=@n6, collidableOnly:bool, DoCameraMove:bool=true) -> ("  ":Exec, Loc:Vector=~, Rot:Rotator=~) {is=title K2Node_MacroInstance}
   > SetSelectedLevelActors(self=Get EditorActorSubsystem(), ActorsToSelect=@n6)
   n7 = Cast To ActorComponent!(execute=@n5.CastFailed, Object=@n2.Object_10_#12) -> (then, CastFailed, AsActor Component:Actor Component Object Reference) {is=title K2Node_DynamicCast}
   n8 = Make Array("[0]"=GetOwner(self=@n7))
   > Focus Actors in Viewport(actors=@n8, DoCameraMove=true) -> (Loc=~, Rot=~)
   > SetSelectedLevelActors(self=Get EditorActorSubsystem(), ActorsToSelect=@n8)
   "On Clicked (btnDeleteEntry)"() {is=title K2Node_ComponentBoundEvent}
   > DeleteSelf() {is=title K2Node_CallFunction}
   n9 = Event PreConstruct()
   > Branch(Condition=@n9.IsDesignTime)
   > ClearReports() {is=title K2Node_CallFunction}
   > For Loop(LastIndex=2)
   > AddReport!(execute, self, ReportData:"Struct GART Report Output Data Structure (by ref)") -> (then) {is=title K2Node_CallFunction}
   n10 = Event OnListItemObjectSet()
   > n11 = Cast To GART_ListData!(execute, Object=@n10.ListItemObject) -> (then, CastFailed, AsGART List Data:GART List Data Object Reference) {is=title K2Node_DynamicCast}
   > Set CurrentReportData(CurrentReportData:Struct GART Asset Report List Structure=Get ReportData!(self:GART List Data Object Reference=@n11) -> (ReportData:Struct GART Asset Report List Structure)) -> (Output_Get:Struct GART Asset Report List Structure)
   > ValidateTexts()
   > ValidateReportItems()
   > ValidateButtonState()
   > SetSelectedReport!(
    execute
    self
    SelectedReport:EUW GART List Item Report Object Reference=Get SelectedReport!(self:GART List Data Object Reference=@n11) -> (SelectedReport:EUW GART List Item Report Object Reference)
   ) -> (then)
   n12 = Event On Item Selection Changed!() -> (OutputDelegate, then, bIsSelected:bool) {is=title K2Node_Event}
   > n13 = Branch(Condition=@n12.bIsSelected)
   > ValidateSelectedVisuals()
   SetSelectedReport(execute=@n13.else)
   ValidateSelectedVisuals()
   Event Construct()
   Event Tick()
  ]}
 HandleReportClicked {nodes [
   n1 = HandleReportClicked!() -> (then, ReportWidget:EUW GART List Item Report Object Reference) {is=title K2Node_FunctionEntry}
   > SetSelectedReport(SelectedReport=@n1)
  ]}
 HandleReportWidgetDeleted {nodes [
   n1 = HandleReportWidgetDeleted!() -> (then, widget:EUW GART List Item Report Object Reference) {is=title K2Node_FunctionEntry}
   > n2 = RemoveChild(self=getReportPanel(), Content=@n1)
   > Branch(Condition=@n2)
   n3 = getReportPanel()
   > Branch(Condition="integer < integer"(A=GetChildrenCount!(self:Panel Widget Object Reference=@n3) -> (ReturnValue:int=0), B=1))
   > DeleteSelf()
  ]}
 SetSelectedReport {nodes [
   n1 = SetSelectedReport!() -> (then, SelectedReport:EUW GART List Item Report Object Reference) {is=title K2Node_FunctionEntry}
   > n2 = Set CurrentSelectedReport(CurrentSelectedReport:EUW GART List Item Report Object Reference=@n1) -> (Output_Get:EUW GART List Item Report Object Reference)
   > n3 = Cast To GART_ListData(Object=GetListItemObject())
   > Set SelectedReport!(execute, SelectedReport:EUW GART List Item Report Object Reference=@n2, self:GART List Data Object Reference=@n3) -> (then, Output_Get:EUW GART List Item Report Object Reference)
   > n4 = Sequence!(execute) -> (then_0, then_1, then_2:Exec)
   > n5 = For Each Loop(Array:Array of Widget Object References=GetAllChildren(self=getReportPanel())) -> (Array Element:Widget Object Reference)
   > n6 = Cast To EUW_GART_ListItem_Report!(execute, Object=@n5) -> (then, CastFailed, AsEUW GART List Item Report:EUW GART List Item Report Object Reference) {is=title K2Node_DynamicCast}
   > SetSelectedState!(
    execute
    self:EUW GART List Item Report Object Reference=@n6
    selected:bool="Equal (Object)"(A=@n6, B=Get SelectedReport!() -> (SelectedReport))
   ) -> (then)
   n7 = Get SelectedReport()
   n8 = Is Valid(exec=@n4.then_1, InputObject=@n7)
   n9 = Get txtDetails() -> (txtDetails:Editor Utility Multi Line Editable Text Box Object Reference)
   SetVisibility(execute=@n8.Is Not Valid, self=@n9, InVisibility=Collapsed)
   n10 = Break Struct GART Report Output Data(Struct_GART_ReportOutputData=Get CurrentReportData!(self:EUW GART List Item Report Object Reference=@n7) -> (CurrentReportData:Struct GART Report Output Data Structure))
   "SetText (Multi-Line Text Box)"!(
    execute=@n8.Is Valid
    self:"Text Box (Multi- Line) Object Reference"=@n9
    InText:Text="To Text (String)"(InString=Break Struct GART Report Data(Struct_GART_ReportData=@n10) -> (Severity_46_#1=~).DisplayDescLong_30_#7)
   ) -> (then)
   > SetVisibility(self=@n9, InVisibility=SelfHitTestInvisible)
   ValidateSelectedVisuals(execute=@n4.then_2)
  ]}
 ValidateButtonState {nodes [
   ValidateButtonState() {is=title K2Node_FunctionEntry}
   n1 = Break Struct GART Asset Report List(Struct_GART_AssetReportList=Get CurrentReportData!() -> (CurrentReportData:Struct GART Asset Report List Structure))
   > SetVisibility(
    self=Get btnEditAsset() -> (btnEditAsset:Editor Utility Button Object Reference)
    InVisibility=Select(Option 0=Visible, Option 1=Collapsed, Index=IsValid(Object=@n1.Object_10_#12))
   )
  ]}
 ValidateReportItems {nodes [
   ValidateReportItems() {is=title K2Node_FunctionEntry}
   > ClearReports()
   > n1 = For Each Loop(Array:Array of Struct GART Report Output Data Structures=Break Struct GART Asset Report List(Struct_GART_AssetReportList=Get CurrentReportData())) -> (Array Element:Struct GART Report Output Data Structure)
   > AddReport(ReportData=@n1)
  ]}
 ValidateSelectedVisuals {nodes [
   ValidateSelectedVisuals() {is=title K2Node_FunctionEntry}
   n1 = IsListItemSelected!(UserListEntry:User List Entry Interface) -> (ReturnValue:bool)
   > SetVisibility(
    self=Get imgBottomEdge() -> (imgBottomEdge:Image Object Reference)
    InVisibility=Select(Option 0=Collapsed, Option 1=HitTestInvisible, Index=@n1)
   )
   > SetBrushColor(
    self=Get borderBG() -> (borderBG:Border Object Reference)
    InBrushColor=Select(
     Option 0:Linear Color Structure=(R=0.0431,G=0.0431,B=0.0431,A=1)
     Option 1:Linear Color Structure=(R=0.1406,G=0.1406,B=0.1406,A=1)
     Index=@n1
    ) -> (ReturnValue:Linear Color Structure)
   )
   n2 = IsValid(Object=Get CurrentSelectedReport() -> (CurrentSelectedReport:EUW GART List Item Report Object Reference))
   n3 = IsListItemSelected()
   > SetVisibility(
    self=Get vBoxDetails() -> (vBoxDetails:Vertical Box Object Reference)
    InVisibility=Select(Option 0:ESlateVisibility Enum=Collapsed, Option 1:ESlateVisibility Enum=SelfHitTestInvisible, Index=AND Boolean(A=@n3, B=@n2)) -> (ReturnValue:ESlateVisibility Enum)
   )
  ]}
 ValidateTexts {nodes [
   ValidateTexts() {is=title K2Node_FunctionEntry}
   n1 = Break Struct GART Asset Report List(Struct_GART_AssetReportList=Get CurrentReportData())
   > n2 = Is Valid(InputObject=@n1.Object_10_#12)
   n3 = Break Asset Data(AssetData=@n1.AssetData_7_#11)
   n4 = Get txtAssetName() -> (txtAssetName:Editor Utility Multi Line Editable Text Box Object Reference)
   "SetText (Multi-Line Text Box)"(execute=@n2.Is Not Valid, self=@n4, InText="To Text (Name)"!(InName:Name=@n3.AssetName) -> (ReturnValue:Text))
   > SetToolTipText!(execute, self:Widget Object Reference=@n4, InToolTipText:"Text (by ref)"="To Text (Name)"(InName=@n3)) -> (then)
   n5 = Break Struct GART Asset Report List(Struct_GART_AssetReportList=Get CurrentReportData())
   "SetText (Multi-Line Text Box)"(execute=@n2.Is Valid, self=@n4, InText="To Text (String)"(InString=GetDisplayName(Object=@n5.Object_10_#12)))
   > SetToolTipText(
    self=@n4
    InToolTipText="To Text (Object)"!(InObj:Object Reference=GetOuterObject!(Object:Object Reference=@n5.Object_10_#12) -> (ReturnValue:Object Reference)) -> (ReturnValue:Text)
   )
   > SetToolTipText(
    self=Get btnGoToAsset() -> (btnGoToAsset:Editor Utility Button Object Reference)
    InToolTipText=MakeLiteralText(Value="INVTEXT(\"Find Actor in Level\")")
   )
   Get txtAssetName()
  ]}
 getReportPanel {nodes [
   getReportPanel() {is=title K2Node_FunctionEntry}
   > Return Node!(
    execute
    panel:Panel Widget Object Reference=Get EditorUtilityScrollBox_202() -> (EditorUtilityScrollBox_202:Editor Utility Scroll Box Object Reference)
   )
  ]}
]
variables [
 CurrentReportData {type=Struct_GART_AssetReportList, default {object=null, assetData {}, reportDatas []}}
 CurrentSelectedReport {type=EUW_GART_ListItem_Report_C, default=null}
]
widgets {
 info {rootWidgetClass=&StackBox, widgetCount=20}
 widgets [
  StackBox_0 {widgetClassPath=&StackBox, widget=StackBox_0, properties {orientation=Orient_Vertical}}
  borderBG {
   widgetClassPath=&Border
   bIsVariable=true
   parent=StackBox_0
   widget=borderBG
   properties {brushColor=<rgba 0.0417 0.0417 0.0417 1>, padding=<ltrb 3 3 3 3>}
   slot {class=StackBoxSlot, properties {size {sizeRule=Fill, value=1}}}
  }
  HorizontalBox_0 {
   widgetClassPath=&HorizontalBox
   parent=borderBG
   widget=HorizontalBox_0
   slot {class=BorderSlot, properties {padding=<ltrb 3 3 3 3>}}
  }
  VerticalBox_0 {widgetClassPath=&VerticalBox, parent=HorizontalBox_0, widget=VerticalBox_0, slot {class=HorizontalBoxSlot}}
  btnEditAsset {
   widgetClassPath=&EditorUtilityButton
   parent=VerticalBox_0
   widget=btnEditAsset
   properties {
    toolTipText=Open this asset for editing
    widgetStyle {
     clickedSlateSound {resourceObject=null}
     disabled {
      bIsDynamicallyLoaded=false
      drawAs=RoundedBox
      imageType=NoImage
      mirroring=NoMirror
      resourceName=None
      resourceObject=null
      tiling=NoTile
      imageSize=<0 0>
      margin=<ltrb 0 0 0 0>
      outlineSettings {
       bUseBrushTransparency=false
       roundingType=FixedRadius
       width=1
       color {colorUseRule=UseColor_ColorTable, specifiedColor=<rgba 0 0 0 0>}
       cornerRadii=<4 4 4 4>
      }
      tintColor {colorUseRule=UseColor_Specified, specifiedColor=<rgba 0.0395 0.0395 0.0395 1>}
      uVRegion {bIsValid=false, max=<0 0>, min=<0 0>}
     }
     disabledForeground {colorUseRule=UseColor_ColorTable, specifiedColor=<rgba 0 0 0 0>}
     hovered {
      bIsDynamicallyLoaded=false
      drawAs=RoundedBox
      imageType=NoImage
      mirroring=NoMirror
      resourceName=None
      resourceObject=null
      tiling=NoTile
      imageSize=<32 32>
      margin=<ltrb 0 0 0 0>
      outlineSettings {
       bUseBrushTransparency=false
       roundingType=FixedRadius
       width=1
       color {colorUseRule=UseColor_ColorTable, specifiedColor=<rgba 0 0 0 0>}
       cornerRadii=<4 4 4 4>
      }
      tintColor {colorUseRule=UseColor_Specified, specifiedColor=<rgba 0.0953 0.0953 0.0953 1>}
      uVRegion {bIsValid=false, max=<0 0>, min=<0 0>}
     }
     hoveredForeground {colorUseRule=UseColor_ColorTable, specifiedColor=<rgba 0 0 0 0>}
     hoveredSlateSound {resourceObject=null}
     normal {
      bIsDynamicallyLoaded=false
      drawAs=RoundedBox
      imageType=NoImage
      mirroring=NoMirror
      resourceName=None
      resourceObject=null
      tiling=NoTile
      imageSize=<32 32>
      margin=<ltrb 0 0 0 0>
      outlineSettings {
       bUseBrushTransparency=false
       roundingType=FixedRadius
       width=1
       color {colorUseRule=UseColor_ColorTable, specifiedColor=<rgba 0 0 0 0>}
       cornerRadii=<4 4 4 4>
      }
      tintColor {colorUseRule=UseColor_Specified, specifiedColor=<rgba 0.0395 0.0395 0.0395 1>}
      uVRegion {bIsValid=false, max=<0 0>, min=<0 0>}
     }
     normalForeground {colorUseRule=UseColor_ColorTable, specifiedColor=<rgba 0 0 0 0>}
     normalPadding=<ltrb 2 2 2 2>
     pressed {
      bIsDynamicallyLoaded=false
      drawAs=RoundedBox
      imageType=NoImage
      mirroring=NoMirror
      resourceName=None
      resourceObject=null
      tiling=NoTile
      imageSize=<32 32>
      margin=<ltrb 0 0 0 0>
      outlineSettings {
       bUseBrushTransparency=false
       roundingType=FixedRadius
       width=1
       color {colorUseRule=UseColor_ColorTable, specifiedColor=<rgba 0 0 0 0>}
       cornerRadii=<4 4 4 4>
      }
      tintColor {colorUseRule=UseColor_Specified, specifiedColor=<rgba 0.0284 0.0284 0.0284 1>}
      uVRegion {bIsValid=false, max=<0 0>, min=<0 0>}
     }
     pressedForeground {colorUseRule=UseColor_ColorTable, specifiedColor=<rgba 0 0 0 0>}
     pressedPadding=<ltrb 3 3 3 3>
     pressedSlateSound {resourceObject=null}
    }
   }
   slot {class=VerticalBoxSlot, properties {size {sizeRule=Fill, value=1}}}
  }
  Image_91 {
   widgetClassPath=&Image
   bIsVariable=true
   parent=btnEditAsset
   widget=Image_91
   properties {brush {
     bIsDynamicallyLoaded=false
     drawAs=Image
     imageType=FullColor
     mirroring=NoMirror
     resourceName=None
     resourceObject=/Engine/EditorResources/S_Note
     tiling=NoTile
     imageSize=<16 16>
     margin=<ltrb 0 0 0 0>
     outlineSettings {
      bUseBrushTransparency=false
      roundingType=HalfHeightRadius
      width=0
      color {colorUseRule=UseColor_Specified, specifiedColor=<rgba 0 0 0 0>}
      cornerRadii=<0 0 0 0>
     }
     tintColor {colorUseRule=UseColor_Specified, specifiedColor=<rgba 1 1 1 1>}
     uVRegion {bIsValid=false, max=<0 0>, min=<0 0>}
    }}
   slot {class=ButtonSlot, properties {padding=<ltrb 0 0 0 0>}}
  }
  btnGoToAsset {
   widgetClassPath=&EditorUtilityButton
   parent=VerticalBox_0
   widget=btnGoToAsset
   properties {
    toolTipText=Go to asset in content browser
    widgetStyle {
     clickedSlateSound {resourceObject=null}
     disabled {
      bIsDynamicallyLoaded=false
      drawAs=RoundedBox
      imageType=NoImage
      mirroring=NoMirror
      resourceName=None
      resourceObject=null
      tiling=NoTile
      imageSize=<0 0>
      margin=<ltrb 0 0 0 0>
      outlineSettings {
       bUseBrushTransparency=false
       roundingType=FixedRadius
       width=1
       color {colorUseRule=UseColor_ColorTable, specifiedColor=<rgba 0 0 0 0>}
       cornerRadii=<4 4 4 4>
      }
      tintColor {colorUseRule=UseColor_Specified, specifiedColor=<rgba 0.0395 0.0395 0.0395 1>}
      uVRegion {bIsValid=false, max=<0 0>, min=<0 0>}
     }
     disabledForeground {colorUseRule=UseColor_ColorTable, specifiedColor=<rgba 0 0 0 0>}
     hovered {
      bIsDynamicallyLoaded=false
      drawAs=RoundedBox
      imageType=NoImage
      mirroring=NoMirror
      resourceName=None
      resourceObject=null
      tiling=NoTile
      imageSize=<32 32>
      margin=<ltrb 0 0 0 0>
      outlineSettings {
       bUseBrushTransparency=false
       roundingType=FixedRadius
       width=1
       color {colorUseRule=UseColor_ColorTable, specifiedColor=<rgba 0 0 0 0>}
       cornerRadii=<4 4 4 4>
      }
      tintColor {colorUseRule=UseColor_Specified, specifiedColor=<rgba 0.0953 0.0953 0.0953 1>}
      uVRegion {bIsValid=false, max=<0 0>, min=<0 0>}
     }
     hoveredForeground {colorUseRule=UseColor_ColorTable, specifiedColor=<rgba 0 0 0 0>}
     hoveredSlateSound {resourceObject=null}
     normal {
      bIsDynamicallyLoaded=false
      drawAs=RoundedBox
      imageType=NoImage
      mirroring=NoMirror
      resourceName=None
      resourceObject=null
      tiling=NoTile
      imageSize=<32 32>
      margin=<ltrb 0 0 0 0>
      outlineSettings {
       bUseBrushTransparency=false
       roundingType=FixedRadius
       width=1
       color {colorUseRule=UseColor_ColorTable, specifiedColor=<rgba 0 0 0 0>}
       cornerRadii=<4 4 4 4>
      }
      tintColor {colorUseRule=UseColor_Specified, specifiedColor=<rgba 0.0395 0.0395 0.0395 1>}
      uVRegion {bIsValid=false, max=<0 0>, min=<0 0>}
     }
     normalForeground {colorUseRule=UseColor_ColorTable, specifiedColor=<rgba 0 0 0 0>}
     normalPadding=<ltrb 2 2 2 2>
     pressed {
      bIsDynamicallyLoaded=false
      drawAs=RoundedBox
      imageType=NoImage
      mirroring=NoMirror
      resourceName=None
      resourceObject=null
      tiling=NoTile
      imageSize=<32 32>
      margin=<ltrb 0 0 0 0>
      outlineSettings {
       bUseBrushTransparency=false
       roundingType=FixedRadius
       width=1
       color {colorUseRule=UseColor_ColorTable, specifiedColor=<rgba 0 0 0 0>}
       cornerRadii=<4 4 4 4>
      }
      tintColor {colorUseRule=UseColor_Specified, specifiedColor=<rgba 0.0284 0.0284 0.0284 1>}
      uVRegion {bIsValid=false, max=<0 0>, min=<0 0>}
     }
     pressedForeground {colorUseRule=UseColor_ColorTable, specifiedColor=<rgba 0 0 0 0>}
     pressedPadding=<ltrb 3 3 3 3>
     pressedSlateSound {resourceObject=null}
    }
   }
   slot {class=VerticalBoxSlot, properties {size {sizeRule=Fill, value=1}}}
  }
  Image {
   widgetClassPath=&Image
   bIsVariable=true
   parent=btnGoToAsset
   widget=Image
   properties {
    brush {
     bIsDynamicallyLoaded=false
     drawAs=Image
     imageType=FullColor
     mirroring=NoMirror
     resourceName=None
     resourceObject=/Engine/EditorResources/Spawn_Point
     tiling=NoTile
     imageSize=<16 16>
     margin=<ltrb 0 0 0 0>
     outlineSettings {
      bUseBrushTransparency=false
      roundingType=HalfHeightRadius
      width=0
      color {colorUseRule=UseColor_Specified, specifiedColor=<rgba 0 0 0 0>}
      cornerRadii=<0 0 0 0>
     }
     tintColor {colorUseRule=UseColor_Specified, specifiedColor=<rgba 1 1 1 1>}
     uVRegion {bIsValid=false, max=<0 0>, min=<0 0>}
    }
    renderTransform {angle=90, scale=<1 1>, shear=<0 0>, translation=<0 0>}
   }
   slot {class=ButtonSlot, properties {padding=<ltrb 0 0 0 0>}}
  }
  ScaleBox_0 {
   widgetClassPath=&ScaleBox
   parent=HorizontalBox_0
   widget=ScaleBox_0
   properties {stretchDirection=DownOnly}
   slot {class=HorizontalBoxSlot, properties {padding=<ltrb 3 3 3 3>, size {sizeRule=Fill, value=0.4}}}
  }
  txtAssetName {
   widgetClassPath=/Script/Blutility.EditorUtilityMultiLineEditableTextBox
   bIsVariable=true
   parent=ScaleBox_0
   widget=txtAssetName
   properties {
    bIsReadOnly=true
    text=Asset Name
    widgetStyle {
     backgroundColor {colorUseRule=UseColor_Specified, specifiedColor=<rgba 1 1 1 1>}
     backgroundImageFocused {
      bIsDynamicallyLoaded=false
      drawAs=RoundedBox
      imageType=NoImage
      mirroring=NoMirror
      resourceName=None
      resourceObject=null
      tiling=NoTile
      imageSize=<0 0>
      margin=<ltrb 0 0 0 0>
      outlineSettings {
       bUseBrushTransparency=false
       roundingType=FixedRadius
       width=1
       color {colorUseRule=UseColor_ColorTable, specifiedColor=<rgba 0 0 0 0>}
       cornerRadii=<4 4 4 4>
      }
      tintColor {colorUseRule=UseColor_Specified, specifiedColor=<rgba 0.0048 0.0048 0.0048 0>}
      uVRegion {bIsValid=false, max=<0 0>, min=<0 0>}
     }
     backgroundImageHovered {
      bIsDynamicallyLoaded=false
      drawAs=RoundedBox
      imageType=NoImage
      mirroring=NoMirror
      resourceName=None
      resourceObject=null
      tiling=NoTile
      imageSize=<0 0>
      margin=<ltrb 0 0 0 0>
      outlineSettings {
       bUseBrushTransparency=false
       roundingType=FixedRadius
       width=1
       color {colorUseRule=UseColor_ColorTable, specifiedColor=<rgba 0 0 0 0>}
       cornerRadii=<4 4 4 4>
      }
      tintColor {colorUseRule=UseColor_Specified, specifiedColor=<rgba 0.0048 0.0048 0.0048 0>}
      uVRegion {bIsValid=false, max=<0 0>, min=<0 0>}
     }
     backgroundImageNormal {
      bIsDynamicallyLoaded=false
      drawAs=RoundedBox
      imageType=NoImage
      mirroring=NoMirror
      resourceName=None
      resourceObject=null
      tiling=NoTile
      imageSize=<0 0>
      margin=<ltrb 0 0 0 0>
      outlineSettings {
       bUseBrushTransparency=false
       roundingType=FixedRadius
       width=1
       color {colorUseRule=UseColor_ColorTable, specifiedColor=<rgba 0 0 0 0>}
       cornerRadii=<4 4 4 4>
      }
      tintColor {colorUseRule=UseColor_Specified, specifiedColor=<rgba 0.0048 0.0048 0.0048 0>}
      uVRegion {bIsValid=false, max=<0 0>, min=<0 0>}
     }
     backgroundImageReadOnly {
      bIsDynamicallyLoaded=false
      drawAs=RoundedBox
      imageType=NoImage
      mirroring=NoMirror
      resourceName=None
      resourceObject=null
      tiling=NoTile
      imageSize=<0 0>
      margin=<ltrb 0 0 0 0>
      outlineSettings {
       bUseBrushTransparency=false
       roundingType=FixedRadius
       width=1
       color {colorUseRule=UseColor_ColorTable, specifiedColor=<rgba 0 0 0 0>}
       cornerRadii=<4 4 4 4>
      }
      tintColor {colorUseRule=UseColor_Specified, specifiedColor=<rgba 0.0048 0.0048 0.0048 0>}
      uVRegion {bIsValid=false, max=<0 0>, min=<0 0>}
     }
     focusedForegroundColor {colorUseRule=UseColor_Specified, specifiedColor=<rgba 1 1 1 1>}
     foregroundColor {colorUseRule=UseColor_Specified, specifiedColor=<rgba 0.5271 0.5271 0.5271 1>}
     hScrollBarPadding=<ltrb 0 0 0 0>
     padding=<ltrb 8 3 8 4>
     readOnlyForegroundColor {colorUseRule=UseColor_Foreground, specifiedColor=<rgba 1 0 1 1>}
     scrollBarStyle {
      thickness=8
      draggedThumbImage {
       bIsDynamicallyLoaded=false
       drawAs=RoundedBox
       imageType=NoImage
       mirroring=NoMirror
       resourceName=None
       resourceObject=null
       tiling=NoTile
       imageSize=<0 0>
       margin=<ltrb 0 0 0 0>
       outlineSettings {
        bUseBrushTransparency=false
        roundingType=FixedRadius
        width=0
        color {colorUseRule=UseColor_Specified, specifiedColor=<rgba 0 0 0 0>}
        cornerRadii=<4 4 4 4>
       }
       tintColor {colorUseRule=UseColor_Specified, specifiedColor=<rgba 0.2159 0.2159 0.2159 1>}
       uVRegion {bIsValid=false, max=<0 0>, min=<0 0>}
      }
      horizontalBackgroundImage {
       bIsDynamicallyLoaded=false
       drawAs=NoDrawType
       imageType=NoImage
       mirroring=NoMirror
       resourceName=None
       resourceObject=null
       tiling=NoTile
       imageSize=<0 0>
       margin=<ltrb 0 0 0 0>
       outlineSettings {
        bUseBrushTransparency=false
        roundingType=HalfHeightRadius
        width=0
        color {colorUseRule=UseColor_Specified, specifiedColor=<rgba 0 0 0 0>}
        cornerRadii=<0 0 0 0>
       }
       tintColor {colorUseRule=UseColor_Specified, specifiedColor=<rgba 1 1 1 1>}
       uVRegion {bIsValid=false, max=<0 0>, min=<0 0>}
      }
      horizontalBottomSlotImage {
       bIsDynamicallyLoaded=false
       drawAs=NoDrawType
       imageType=NoImage
       mirroring=NoMirror
       resourceName=None
       resourceObject=null
       tiling=NoTile
       imageSize=<0 0>
       margin=<ltrb 0 0 0 0>
       outlineSettings {
        bUseBrushTransparency=false
        roundingType=HalfHeightRadius
        width=0
        color {colorUseRule=UseColor_Specified, specifiedColor=<rgba 0 0 0 0>}
        cornerRadii=<0 0 0 0>
       }
       tintColor {colorUseRule=UseColor_Specified, specifiedColor=<rgba 1 1 1 1>}
       uVRegion {bIsValid=false, max=<0 0>, min=<0 0>}
      }
      horizontalTopSlotImage {
       bIsDynamicallyLoaded=false
       drawAs=NoDrawType
       imageType=NoImage
       mirroring=NoMirror
       resourceName=None
       resourceObject=null
       tiling=NoTile
       imageSize=<0 0>
       margin=<ltrb 0 0 0 0>
       outlineSettings {
        bUseBrushTransparency=false
        roundingType=HalfHeightRadius
        width=0
        color {colorUseRule=UseColor_Specified, specifiedColor=<rgba 0 0 0 0>}
        cornerRadii=<0 0 0 0>
       }
       tintColor {colorUseRule=UseColor_Specified, specifiedColor=<rgba 1 1 1 1>}
       uVRegion {bIsValid=false, max=<0 0>, min=<0 0>}
      }
      hoveredThumbImage {
       bIsDynamicallyLoaded=false
       drawAs=RoundedBox
       imageType=NoImage
       mirroring=NoMirror
       resourceName=None
       resourceObject=null
       tiling=NoTile
       imageSize=<0 0>
       margin=<ltrb 0 0 0 0>
       outlineSettings {
        bUseBrushTransparency=false
        roundingType=FixedRadius
        width=0
        color {colorUseRule=UseColor_Specified, specifiedColor=<rgba 0 0 0 0>}
        cornerRadii=<4 4 4 4>
       }
       tintColor {colorUseRule=UseColor_Specified, specifiedColor=<rgba 0.2159 0.2159 0.2159 1>}
       uVRegion {bIsValid=false, max=<0 0>, min=<0 0>}
      }
      normalThumbImage {
       bIsDynamicallyLoaded=false
       drawAs=RoundedBox
       imageType=NoImage
       mirroring=NoMirror
       resourceName=None
       resourceObject=null
       tiling=NoTile
       imageSize=<0 0>
       margin=<ltrb 0 0 0 0>
       outlineSettings {
        bUseBrushTransparency=false
        roundingType=FixedRadius
        width=0
        color {colorUseRule=UseColor_Specified, specifiedColor=<rgba 0 0 0 0>}
        cornerRadii=<4 4 4 4>
       }
       tintColor {colorUseRule=UseColor_Specified, specifiedColor=<rgba 0.0953 0.0953 0.0953 1>}
       uVRegion {bIsValid=false, max=<0 0>, min=<0 0>}
      }
      verticalBackgroundImage {
       bIsDynamicallyLoaded=false
       drawAs=NoDrawType
       imageType=NoImage
       mirroring=NoMirror
       resourceName=None
       resourceObject=null
       tiling=NoTile
       imageSize=<0 0>
       margin=<ltrb 0 0 0 0>
       outlineSettings {
        bUseBrushTransparency=false
        roundingType=HalfHeightRadius
        width=0
        color {colorUseRule=UseColor_Specified, specifiedColor=<rgba 0 0 0 0>}
        cornerRadii=<0 0 0 0>
       }
       tintColor {colorUseRule=UseColor_Specified, specifiedColor=<rgba 1 1 1 1>}
       uVRegion {bIsValid=false, max=<0 0>, min=<0 0>}
      }
      verticalBottomSlotImage {
       bIsDynamicallyLoaded=false
       drawAs=NoDrawType
       imageType=NoImage
       mirroring=NoMirror
       resourceName=None
       resourceObject=null
       tiling=NoTile
       imageSize=<0 0>
       margin=<ltrb 0 0 0 0>
       outlineSettings {
        bUseBrushTransparency=false
        roundingType=HalfHeightRadius
        width=0
        color {colorUseRule=UseColor_Specified, specifiedColor=<rgba 0 0 0 0>}
        cornerRadii=<0 0 0 0>
       }
       tintColor {colorUseRule=UseColor_Specified, specifiedColor=<rgba 1 1 1 1>}
       uVRegion {bIsValid=false, max=<0 0>, min=<0 0>}
      }
      verticalTopSlotImage {
       bIsDynamicallyLoaded=false
       drawAs=NoDrawType
       imageType=NoImage
       mirroring=NoMirror
       resourceName=None
       resourceObject=null
       tiling=NoTile
       imageSize=<0 0>
       margin=<ltrb 0 0 0 0>
       outlineSettings {
        bUseBrushTransparency=false
        roundingType=HalfHeightRadius
        width=0
        color {colorUseRule=UseColor_Specified, specifiedColor=<rgba 0 0 0 0>}
        cornerRadii=<0 0 0 0>
       }
       tintColor {colorUseRule=UseColor_Specified, specifiedColor=<rgba 1 1 1 1>}
       uVRegion {bIsValid=false, max=<0 0>, min=<0 0>}
      }
     }
     textStyle {
      overflowPolicy=Clip
      transformPolicy=None
      colorAndOpacity {colorUseRule=UseColor_Foreground, specifiedColor=<rgba 1 0 1 1>}
      font {
       bForceMonospaced=false
       bMaterialIsStencil=false
       fontMaterial=null
       fontObject=&NotoSans
       letterSpacing=0
       monospacedWidth=1
       size=10.5
       skewAmount=0
       typefaceFontName=NotoSans-Regular
       outlineSettings {
        bApplyOutlineToDropShadows=false
        bMiteredCorners=false
        bSeparateFillAlpha=false
        outlineMaterial=null
        outlineSize=0
        outlineColor=<rgba 0 0 0 1>
       }
      }
      highlightColor {colorUseRule=UseColor_ColorTable, specifiedColor=<rgba 0 0 0 0>}
      highlightShape {
       bIsDynamicallyLoaded=false
       drawAs=Image
       imageType=NoImage
       mirroring=NoMirror
       resourceName=None
       resourceObject=null
       tiling=NoTile
       imageSize=<0 0>
       margin=<ltrb 0 0 0 0>
       outlineSettings {
        bUseBrushTransparency=false
        roundingType=HalfHeightRadius
        width=0
        color {colorUseRule=UseColor_Specified, specifiedColor=<rgba 0 0 0 0>}
        cornerRadii=<0 0 0 0>
       }
       tintColor {colorUseRule=UseColor_ColorTable, specifiedColor=<rgba 0 0 0 0>}
       uVRegion {bIsValid=false, max=<0 0>, min=<0 0>}
      }
      selectedBackgroundColor {colorUseRule=UseColor_ColorTable, specifiedColor=<rgba 0 0 0 0>}
      shadowColorAndOpacity=<rgba 0 0 0 1>
      shadowOffset=<0 0>
      strikeBrush {
       bIsDynamicallyLoaded=false
       drawAs=Image
       imageType=NoImage
       mirroring=NoMirror
       resourceName=None
       resourceObject=null
       tiling=NoTile
       imageSize=<32 32>
       margin=<ltrb 0 0 0 0>
       outlineSettings {
        bUseBrushTransparency=false
        roundingType=HalfHeightRadius
        width=0
        color {colorUseRule=UseColor_Specified, specifiedColor=<rgba 0 0 0 0>}
        cornerRadii=<0 0 0 0>
       }
       tintColor {colorUseRule=UseColor_Specified, specifiedColor=<rgba 1 1 1 1>}
       uVRegion {bIsValid=false, max=<0 0>, min=<0 0>}
      }
      underlineBrush {
       bIsDynamicallyLoaded=false
       drawAs=Image
       imageType=NoImage
       mirroring=NoMirror
       resourceName=None
       resourceObject=null
       tiling=NoTile
       imageSize=<32 32>
       margin=<ltrb 0 0 0 0>
       outlineSettings {
        bUseBrushTransparency=false
        roundingType=HalfHeightRadius
        width=0
        color {colorUseRule=UseColor_Specified, specifiedColor=<rgba 0 0 0 0>}
        cornerRadii=<0 0 0 0>
       }
       tintColor {colorUseRule=UseColor_Specified, specifiedColor=<rgba 1 1 1 1>}
       uVRegion {bIsValid=false, max=<0 0>, min=<0 0>}
      }
     }
     vScrollBarPadding=<ltrb 0 0 0 0>
    }
   }
   slot {class=ScaleBoxSlot, properties {horizontalAlignment=HAlign_Fill}}
  }
  EditorUtilityScrollBox_202 {
   widgetClassPath=&EditorUtilityScrollBox
   bIsVariable=true
   parent=HorizontalBox_0
   widget=EditorUtilityScrollBox_202
   properties {allowOverscroll=false, orientation=Orient_Horizontal, scrollbarThickness=<3 3>}
   slot {class=HorizontalBoxSlot, properties {padding=<ltrb 5 0 0 0>, size {sizeRule=Fill, value=1}}}
  }
  btnDeleteEntry {
   widgetClassPath=&EditorUtilityButton
   parent=HorizontalBox_0
   widget=btnDeleteEntry
   properties {
    toolTipText=Delete this entry
    backgroundColor=<rgba 0.7552 0.2832 0.2832 1>
    colorAndOpacity=<rgba 0.5521 0.069 0.069 1>
    widgetStyle {
     clickedSlateSound {resourceObject=null}
     disabled {
      bIsDynamicallyLoaded=false
      drawAs=RoundedBox
      imageType=NoImage
      mirroring=NoMirror
      resourceName=None
      resourceObject=null
      tiling=NoTile
      imageSize=<0 0>
      margin=<ltrb 0 0 0 0>
      outlineSettings {
       bUseBrushTransparency=false
       roundingType=FixedRadius
       width=1
       color {colorUseRule=UseColor_ColorTable, specifiedColor=<rgba 0 0 0 0>}
       cornerRadii=<4 4 4 4>
      }
      tintColor {colorUseRule=UseColor_Specified, specifiedColor=<rgba 0.0395 0.0395 0.0395 1>}
      uVRegion {bIsValid=false, max=<0 0>, min=<0 0>}
     }
     disabledForeground {colorUseRule=UseColor_ColorTable, specifiedColor=<rgba 0 0 0 0>}
     hovered {
      bIsDynamicallyLoaded=false
      drawAs=RoundedBox
      imageType=NoImage
      mirroring=NoMirror
      resourceName=None
      resourceObject=null
      tiling=NoTile
      imageSize=<32 32>
      margin=<ltrb 0 0 0 0>
      outlineSettings {
       bUseBrushTransparency=false
       roundingType=FixedRadius
       width=1
       color {colorUseRule=UseColor_ColorTable, specifiedColor=<rgba 0 0 0 0>}
       cornerRadii=<4 4 4 4>
      }
      tintColor {colorUseRule=UseColor_Specified, specifiedColor=<rgba 0.0953 0.0953 0.0953 1>}
      uVRegion {bIsValid=false, max=<0 0>, min=<0 0>}
     }
     hoveredForeground {colorUseRule=UseColor_ColorTable, specifiedColor=<rgba 0 0 0 0>}
     hoveredSlateSound {resourceObject=null}
     normal {
      bIsDynamicallyLoaded=false
      drawAs=RoundedBox
      imageType=NoImage
      mirroring=NoMirror
      resourceName=None
      resourceObject=null
      tiling=NoTile
      imageSize=<32 32>
      margin=<ltrb 0 0 0 0>
      outlineSettings {
       bUseBrushTransparency=false
       roundingType=FixedRadius
       width=1
       color {colorUseRule=UseColor_ColorTable, specifiedColor=<rgba 0 0 0 0>}
       cornerRadii=<4 4 4 4>
      }
      tintColor {colorUseRule=UseColor_Specified, specifiedColor=<rgba 0.0395 0.0395 0.0395 1>}
      uVRegion {bIsValid=false, max=<0 0>, min=<0 0>}
     }
     normalForeground {colorUseRule=UseColor_ColorTable, specifiedColor=<rgba 0 0 0 0>}
     normalPadding=<ltrb 2 2 2 2>
     pressed {
      bIsDynamicallyLoaded=false
      drawAs=RoundedBox
      imageType=NoImage
      mirroring=NoMirror
      resourceName=None
      resourceObject=null
      tiling=NoTile
      imageSize=<32 32>
      margin=<ltrb 0 0 0 0>
      outlineSettings {
       bUseBrushTransparency=false
       roundingType=FixedRadius
       width=1
       color {colorUseRule=UseColor_ColorTable, specifiedColor=<rgba 0 0 0 0>}
       cornerRadii=<4 4 4 4>
      }
      tintColor {colorUseRule=UseColor_Specified, specifiedColor=<rgba 0.0284 0.0284 0.0284 1>}
      uVRegion {bIsValid=false, max=<0 0>, min=<0 0>}
     }
     pressedForeground {colorUseRule=UseColor_ColorTable, specifiedColor=<rgba 0 0 0 0>}
     pressedPadding=<ltrb 3 3 3 3>
     pressedSlateSound {resourceObject=null}
    }
   }
   slot {class=HorizontalBoxSlot}
  }
  Image_1 {
   widgetClassPath=&Image
   bIsVariable=true
   parent=btnDeleteEntry
   widget=Image_1
   properties {
    brush {
     bIsDynamicallyLoaded=false
     drawAs=Image
     imageType=FullColor
     mirroring=NoMirror
     resourceName=None
     resourceObject=/Game/UI/Foundation/Textures/Buttons/T-SlotIcon-Remove-32
     tiling=NoTile
     imageSize=<16 16>
     margin=<ltrb 0 0 0 0>
     outlineSettings {
      bUseBrushTransparency=false
      roundingType=HalfHeightRadius
      width=0
      color {colorUseRule=UseColor_Specified, specifiedColor=<rgba 0 0 0 0>}
      cornerRadii=<0 0 0 0>
     }
     tintColor {colorUseRule=UseColor_Specified, specifiedColor=<rgba 1 1 1 1>}
     uVRegion {bIsValid=false, max=<0 0>, min=<0 0>}
    }
    renderTransform {angle=90, scale=<1 1>, shear=<0 0>, translation=<0 0>}
   }
   slot {class=ButtonSlot, properties {padding=<ltrb 0 0 0 0>}}
  }
  imgBottomEdge {
   widgetClassPath=&Image
   bIsVariable=true
   parent=StackBox_0
   widget=imgBottomEdge
   properties {
    brush {
     bIsDynamicallyLoaded=false
     drawAs=Image
     imageType=NoImage
     mirroring=NoMirror
     resourceName=None
     resourceObject=null
     tiling=NoTile
     imageSize=<3 3>
     margin=<ltrb 0 0 0 0>
     outlineSettings {
      bUseBrushTransparency=false
      roundingType=HalfHeightRadius
      width=0
      color {colorUseRule=UseColor_Specified, specifiedColor=<rgba 0 0 0 0>}
      cornerRadii=<0 0 0 0>
     }
     tintColor {colorUseRule=UseColor_Specified, specifiedColor=<rgba 1 1 1 1>}
     uVRegion {bIsValid=false, max=<0 0>, min=<0 0>}
    }
    colorAndOpacity=<rgba 0.2135 0.2135 0.2135 1>
   }
   slot {class=StackBoxSlot}
  }
  Border_0 {
   widgetClassPath=&Border
   parent=StackBox_0
   widget=Border_0
   properties {brushColor=<rgba 0.01 0.01 0.01 1>, padding=<ltrb 0 0 0 0>}
   slot {class=StackBoxSlot, properties {verticalAlignment=VAlign_Center, padding=<ltrb 1 1 1 1>}}
  }
  vBoxDetails {
   widgetClassPath=&VerticalBox
   bIsVariable=true
   parent=Border_0
   widget=vBoxDetails
   properties {visibility=Collapsed}
   slot {class=BorderSlot, properties {padding=<ltrb 0 0 0 0>}}
  }
  SizeBox_0 {
   widgetClassPath=&SizeBox
   parent=vBoxDetails
   widget=SizeBox_0
   properties {bOverride_MaxDesiredHeight=true, heightOverride=52.6268, maxDesiredHeight=75}
   slot {class=VerticalBoxSlot}
  }
  EditorUtilityScrollBox_100 {widgetClassPath=&EditorUtilityScrollBox, parent=SizeBox_0, widget=EditorUtilityScrollBox_100, slot {class=SizeBoxSlot}}
  txtDetails {
   widgetClassPath=&EditorUtilityMultiLineEditableTextBox
   bIsVariable=true
   parent=EditorUtilityScrollBox_100
   widget=txtDetails
   properties {
    bIsReadOnly=true
    widgetStyle {
     backgroundColor {colorUseRule=UseColor_Specified, specifiedColor=<rgba 1 1 1 1>}
     backgroundImageFocused {
      bIsDynamicallyLoaded=false
      drawAs=RoundedBox
      imageType=NoImage
      mirroring=NoMirror
      resourceName=None
      resourceObject=null
      tiling=NoTile
      imageSize=<0 0>
      margin=<ltrb 0 0 0 0>
      outlineSettings {
       bUseBrushTransparency=false
       roundingType=FixedRadius
       width=1
       color {colorUseRule=UseColor_ColorTable, specifiedColor=<rgba 0 0 0 0>}
       cornerRadii=<4 4 4 4>
      }
      tintColor {colorUseRule=UseColor_Specified, specifiedColor=<rgba 0.0048 0.0048 0.0048 0>}
      uVRegion {bIsValid=false, max=<0 0>, min=<0 0>}
     }
     backgroundImageHovered {
      bIsDynamicallyLoaded=false
      drawAs=RoundedBox
      imageType=NoImage
      mirroring=NoMirror
      resourceName=None
      resourceObject=null
      tiling=NoTile
      imageSize=<0 0>
      margin=<ltrb 0 0 0 0>
      outlineSettings {
       bUseBrushTransparency=false
       roundingType=FixedRadius
       width=1
       color {colorUseRule=UseColor_ColorTable, specifiedColor=<rgba 0 0 0 0>}
       cornerRadii=<4 4 4 4>
      }
      tintColor {colorUseRule=UseColor_Specified, specifiedColor=<rgba 0.0048 0.0048 0.0048 0>}
      uVRegion {bIsValid=false, max=<0 0>, min=<0 0>}
     }
     backgroundImageNormal {
      bIsDynamicallyLoaded=false
      drawAs=RoundedBox
      imageType=NoImage
      mirroring=NoMirror
      resourceName=None
      resourceObject=null
      tiling=NoTile
      imageSize=<0 0>
      margin=<ltrb 0 0 0 0>
      outlineSettings {
       bUseBrushTransparency=false
       roundingType=FixedRadius
       width=1
       color {colorUseRule=UseColor_ColorTable, specifiedColor=<rgba 0 0 0 0>}
       cornerRadii=<4 4 4 4>
      }
      tintColor {colorUseRule=UseColor_Specified, specifiedColor=<rgba 0.0048 0.0048 0.0048 0>}
      uVRegion {bIsValid=false, max=<0 0>, min=<0 0>}
     }
     backgroundImageReadOnly {
      bIsDynamicallyLoaded=false
      drawAs=RoundedBox
      imageType=NoImage
      mirroring=NoMirror
      resourceName=None
      resourceObject=null
      tiling=NoTile
      imageSize=<0 0>
      margin=<ltrb 0 0 0 0>
      outlineSettings {
       bUseBrushTransparency=false
       roundingType=FixedRadius
       width=1
       color {colorUseRule=UseColor_ColorTable, specifiedColor=<rgba 0 0 0 0>}
       cornerRadii=<4 4 4 4>
      }
      tintColor {colorUseRule=UseColor_Specified, specifiedColor=<rgba 0.0048 0.0048 0.0048 0>}
      uVRegion {bIsValid=false, max=<0 0>, min=<0 0>}
     }
     focusedForegroundColor {colorUseRule=UseColor_Specified, specifiedColor=<rgba 1 1 1 1>}
     foregroundColor {colorUseRule=UseColor_Specified, specifiedColor=<rgba 0.5271 0.5271 0.5271 1>}
     hScrollBarPadding=<ltrb 0 0 0 0>
     padding=<ltrb 8 3 8 4>
     readOnlyForegroundColor {colorUseRule=UseColor_Foreground, specifiedColor=<rgba 1 0 1 1>}
     scrollBarStyle {
      thickness=8
      draggedThumbImage {
       bIsDynamicallyLoaded=false
       drawAs=RoundedBox
       imageType=NoImage
       mirroring=NoMirror
       resourceName=None
       resourceObject=null
       tiling=NoTile
       imageSize=<0 0>
       margin=<ltrb 0 0 0 0>
       outlineSettings {
        bUseBrushTransparency=false
        roundingType=FixedRadius
        width=0
        color {colorUseRule=UseColor_Specified, specifiedColor=<rgba 0 0 0 0>}
        cornerRadii=<4 4 4 4>
       }
       tintColor {colorUseRule=UseColor_Specified, specifiedColor=<rgba 0.2159 0.2159 0.2159 1>}
       uVRegion {bIsValid=false, max=<0 0>, min=<0 0>}
      }
      horizontalBackgroundImage {
       bIsDynamicallyLoaded=false
       drawAs=NoDrawType
       imageType=NoImage
       mirroring=NoMirror
       resourceName=None
       resourceObject=null
       tiling=NoTile
       imageSize=<0 0>
       margin=<ltrb 0 0 0 0>
       outlineSettings {
        bUseBrushTransparency=false
        roundingType=HalfHeightRadius
        width=0
        color {colorUseRule=UseColor_Specified, specifiedColor=<rgba 0 0 0 0>}
        cornerRadii=<0 0 0 0>
       }
       tintColor {colorUseRule=UseColor_Specified, specifiedColor=<rgba 1 1 1 1>}
       uVRegion {bIsValid=false, max=<0 0>, min=<0 0>}
      }
      horizontalBottomSlotImage {
       bIsDynamicallyLoaded=false
       drawAs=NoDrawType
       imageType=NoImage
       mirroring=NoMirror
       resourceName=None
       resourceObject=null
       tiling=NoTile
       imageSize=<0 0>
       margin=<ltrb 0 0 0 0>
       outlineSettings {
        bUseBrushTransparency=false
        roundingType=HalfHeightRadius
        width=0
        color {colorUseRule=UseColor_Specified, specifiedColor=<rgba 0 0 0 0>}
        cornerRadii=<0 0 0 0>
       }
       tintColor {colorUseRule=UseColor_Specified, specifiedColor=<rgba 1 1 1 1>}
       uVRegion {bIsValid=false, max=<0 0>, min=<0 0>}
      }
      horizontalTopSlotImage {
       bIsDynamicallyLoaded=false
       drawAs=NoDrawType
       imageType=NoImage
       mirroring=NoMirror
       resourceName=None
       resourceObject=null
       tiling=NoTile
       imageSize=<0 0>
       margin=<ltrb 0 0 0 0>
       outlineSettings {
        bUseBrushTransparency=false
        roundingType=HalfHeightRadius
        width=0
        color {colorUseRule=UseColor_Specified, specifiedColor=<rgba 0 0 0 0>}
        cornerRadii=<0 0 0 0>
       }
       tintColor {colorUseRule=UseColor_Specified, specifiedColor=<rgba 1 1 1 1>}
       uVRegion {bIsValid=false, max=<0 0>, min=<0 0>}
      }
      hoveredThumbImage {
       bIsDynamicallyLoaded=false
       drawAs=RoundedBox
       imageType=NoImage
       mirroring=NoMirror
       resourceName=None
       resourceObject=null
       tiling=NoTile
       imageSize=<0 0>
       margin=<ltrb 0 0 0 0>
       outlineSettings {
        bUseBrushTransparency=false
        roundingType=FixedRadius
        width=0
        color {colorUseRule=UseColor_Specified, specifiedColor=<rgba 0 0 0 0>}
        cornerRadii=<4 4 4 4>
       }
       tintColor {colorUseRule=UseColor_Specified, specifiedColor=<rgba 0.2159 0.2159 0.2159 1>}
       uVRegion {bIsValid=false, max=<0 0>, min=<0 0>}
      }
      normalThumbImage {
       bIsDynamicallyLoaded=false
       drawAs=RoundedBox
       imageType=NoImage
       mirroring=NoMirror
       resourceName=None
       resourceObject=null
       tiling=NoTile
       imageSize=<0 0>
       margin=<ltrb 0 0 0 0>
       outlineSettings {
        bUseBrushTransparency=false
        roundingType=FixedRadius
        width=0
        color {colorUseRule=UseColor_Specified, specifiedColor=<rgba 0 0 0 0>}
        cornerRadii=<4 4 4 4>
       }
       tintColor {colorUseRule=UseColor_Specified, specifiedColor=<rgba 0.0953 0.0953 0.0953 1>}
       uVRegion {bIsValid=false, max=<0 0>, min=<0 0>}
      }
      verticalBackgroundImage {
       bIsDynamicallyLoaded=false
       drawAs=NoDrawType
       imageType=NoImage
       mirroring=NoMirror
       resourceName=None
       resourceObject=null
       tiling=NoTile
       imageSize=<0 0>
       margin=<ltrb 0 0 0 0>
       outlineSettings {
        bUseBrushTransparency=false
        roundingType=HalfHeightRadius
        width=0
        color {colorUseRule=UseColor_Specified, specifiedColor=<rgba 0 0 0 0>}
        cornerRadii=<0 0 0 0>
       }
       tintColor {colorUseRule=UseColor_Specified, specifiedColor=<rgba 1 1 1 1>}
       uVRegion {bIsValid=false, max=<0 0>, min=<0 0>}
      }
      verticalBottomSlotImage {
       bIsDynamicallyLoaded=false
       drawAs=NoDrawType
       imageType=NoImage
       mirroring=NoMirror
       resourceName=None
       resourceObject=null
       tiling=NoTile
       imageSize=<0 0>
       margin=<ltrb 0 0 0 0>
       outlineSettings {
        bUseBrushTransparency=false
        roundingType=HalfHeightRadius
        width=0
        color {colorUseRule=UseColor_Specified, specifiedColor=<rgba 0 0 0 0>}
        cornerRadii=<0 0 0 0>
       }
       tintColor {colorUseRule=UseColor_Specified, specifiedColor=<rgba 1 1 1 1>}
       uVRegion {bIsValid=false, max=<0 0>, min=<0 0>}
      }
      verticalTopSlotImage {
       bIsDynamicallyLoaded=false
       drawAs=NoDrawType
       imageType=NoImage
       mirroring=NoMirror
       resourceName=None
       resourceObject=null
       tiling=NoTile
       imageSize=<0 0>
       margin=<ltrb 0 0 0 0>
       outlineSettings {
        bUseBrushTransparency=false
        roundingType=HalfHeightRadius
        width=0
        color {colorUseRule=UseColor_Specified, specifiedColor=<rgba 0 0 0 0>}
        cornerRadii=<0 0 0 0>
       }
       tintColor {colorUseRule=UseColor_Specified, specifiedColor=<rgba 1 1 1 1>}
       uVRegion {bIsValid=false, max=<0 0>, min=<0 0>}
      }
     }
     textStyle {
      overflowPolicy=Clip
      transformPolicy=None
      colorAndOpacity {colorUseRule=UseColor_Foreground, specifiedColor=<rgba 1 0 1 1>}
      font {
       bForceMonospaced=false
       bMaterialIsStencil=false
       fontMaterial=null
       fontObject=&NotoSans
       letterSpacing=0
       monospacedWidth=1
       size=9
       skewAmount=0
       typefaceFontName=NotoSans-Regular
       outlineSettings {
        bApplyOutlineToDropShadows=false
        bMiteredCorners=false
        bSeparateFillAlpha=false
        outlineMaterial=null
        outlineSize=0
        outlineColor=<rgba 0 0 0 1>
       }
      }
      highlightColor {colorUseRule=UseColor_ColorTable, specifiedColor=<rgba 0 0 0 0>}
      highlightShape {
       bIsDynamicallyLoaded=false
       drawAs=Image
       imageType=NoImage
       mirroring=NoMirror
       resourceName=None
       resourceObject=null
       tiling=NoTile
       imageSize=<0 0>
       margin=<ltrb 0 0 0 0>
       outlineSettings {
        bUseBrushTransparency=false
        roundingType=HalfHeightRadius
        width=0
        color {colorUseRule=UseColor_Specified, specifiedColor=<rgba 0 0 0 0>}
        cornerRadii=<0 0 0 0>
       }
       tintColor {colorUseRule=UseColor_ColorTable, specifiedColor=<rgba 0 0 0 0>}
       uVRegion {bIsValid=false, max=<0 0>, min=<0 0>}
      }
      selectedBackgroundColor {colorUseRule=UseColor_ColorTable, specifiedColor=<rgba 0 0 0 0>}
      shadowColorAndOpacity=<rgba 0 0 0 1>
      shadowOffset=<0 0>
      strikeBrush {
       bIsDynamicallyLoaded=false
       drawAs=Image
       imageType=NoImage
       mirroring=NoMirror
       resourceName=None
       resourceObject=null
       tiling=NoTile
       imageSize=<32 32>
       margin=<ltrb 0 0 0 0>
       outlineSettings {
        bUseBrushTransparency=false
        roundingType=HalfHeightRadius
        width=0
        color {colorUseRule=UseColor_Specified, specifiedColor=<rgba 0 0 0 0>}
        cornerRadii=<0 0 0 0>
       }
       tintColor {colorUseRule=UseColor_Specified, specifiedColor=<rgba 1 1 1 1>}
       uVRegion {bIsValid=false, max=<0 0>, min=<0 0>}
      }
      underlineBrush {
       bIsDynamicallyLoaded=false
       drawAs=Image
       imageType=NoImage
       mirroring=NoMirror
       resourceName=None
       resourceObject=null
       tiling=NoTile
       imageSize=<32 32>
       margin=<ltrb 0 0 0 0>
       outlineSettings {
        bUseBrushTransparency=false
        roundingType=HalfHeightRadius
        width=0
        color {colorUseRule=UseColor_Specified, specifiedColor=<rgba 0 0 0 0>}
        cornerRadii=<0 0 0 0>
       }
       tintColor {colorUseRule=UseColor_Specified, specifiedColor=<rgba 1 1 1 1>}
       uVRegion {bIsValid=false, max=<0 0>, min=<0 0>}
      }
     }
     vScrollBarPadding=<ltrb 0 0 0 0>
    }
   }
   slot {class=ScrollBoxSlot}
  }
  Image_2 {
   widgetClassPath=&Image
   bIsVariable=true
   parent=vBoxDetails
   widget=Image_2
   properties {
    brush {
     bIsDynamicallyLoaded=false
     drawAs=Image
     imageType=NoImage
     mirroring=NoMirror
     resourceName=None
     resourceObject=null
     tiling=NoTile
     imageSize=<3 3>
     margin=<ltrb 0 0 0 0>
     outlineSettings {
      bUseBrushTransparency=false
      roundingType=HalfHeightRadius
      width=0
      color {colorUseRule=UseColor_Specified, specifiedColor=<rgba 0 0 0 0>}
      cornerRadii=<0 0 0 0>
     }
     tintColor {colorUseRule=UseColor_Specified, specifiedColor=<rgba 1 1 1 1>}
     uVRegion {bIsValid=false, max=<0 0>, min=<0 0>}
    }
    colorAndOpacity=<rgba 1 0.4325 0.0214 1>
   }
   slot {class=VerticalBoxSlot}
  }
 ]
}
