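# Editor utility list panel for the Generic Asset Review Tool, displaying asset and object review reports.
# RefreshList wraps both report maps into GART_ListData items for the list view.
# Entry events toggle the empty message and export buttons, closing the widget's tab when empty.
# Buttons export rows to CSV through inline Python, or save non-object entries' assets to a local collection, persisting its dialog-entered name via SaveGame.
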EditorUtilityWidgetBlueprint /Game/NeverCook/EditorUtilities/GenericAssetReviewTool/Widgets/EUW_GART_List_Main
graphs [
 EventGraph {nodes [
   "On Clicked (btnCollection)"!() -> (OutputDelegate, then, widget:User Widget Object Reference, event:Pointer Event Structure) {is=title K2Node_ComponentBoundEvent}
   > SaveAssetsToCollection()
   n1 = "On Entry Generated (listMain)"!() -> (OutputDelegate, then, Widget:User Widget Object Reference) {is=title K2Node_ComponentBoundEvent}
   n2 = "integer > integer"(A=GetNumItems!(self:List View Object Reference=Get listMain() -> (listMain:Common List View Object Reference)) -> (ReturnValue:int=0))
   SetVisibility(
    execute=[@n1.then, @n6.then]
    self=Get txtNoIssues() -> (txtNoIssues:Common Text Object Reference)
    InVisibility=Select(Option 0=Visible, Option 1=Collapsed, Index=@n2)
   )
   > SetVisibility(
    self=Get panelExportButtons() -> (panelExportButtons:Horizontal Box Object Reference)
    InVisibility=Select(Option 0=Collapsed, Option 1=SelfHitTestInvisible, Index=@n2)
   )
   > n3 = Branch(Condition=@n2)
   n4 = Get EditorUtilitySubsystem()
   n5 = RegisterTabAndGetID!(
    execute=@n3.else
    self:Editor Utility Subsystem Object Reference=@n4
    InBlueprint:Editor Utility Widget Blueprint Object Reference=&self
   ) -> (then, NewTabID:Name=None)
   > CloseTabByID!(execute, self:Editor Utility Subsystem Object Reference=@n4, NewTabID:Name=@n5) -> (then, ReturnValue:bool)
   n6 = "On Entry Released (listMain)"!() -> (OutputDelegate, then, Widget:User Widget Object Reference) {is=title K2Node_ComponentBoundEvent}
   "On Clicked (btnExport)"!() -> (OutputDelegate, then, widget:User Widget Object Reference, event:Pointer Event Structure) {is=title K2Node_ComponentBoundEvent}
   > ExportCSV()
  ]}
 ExportCSV {nodes [
   ExportCSV() {is=title K2Node_FunctionEntry}
   > n1 = Sequence!(execute) -> (then_0, then_1, then_2:Exec)
   > n2 = For Each Loop(Array:Array of Object References=GetListItems!(self:List View Object Reference=Get listMain()) -> (ReturnValue:Array of Object References)) -> (Array Element:Object Reference)
   > n3 = Cast To GART_ListData!(execute, Object=@n2) -> (then, CastFailed, AsGART List Data:GART List Data Object Reference) {is=title K2Node_DynamicCast}
   > n4 = For Each Loop(Array:Array of Struct GART Report Output Data Structures=Break Struct GART Asset Report List!(Struct_GART_AssetReportList:"Struct GART Asset Report List Structure (by ref)"=Get ReportData!(self:GART List Data Object Reference=@n3) -> (ReportData:Struct GART Asset Report List Structure)) -> (ReportDatas_11_#1:Array of Struct GART Report Output Data Structures) {is=title K2Node_BreakStruct}) -> (Array Element:Struct GART Report Output Data Structure)
   > n5 = GetCSVRowForData!(execute, data:Struct GART Report Output Data Structure=@n4, __WorldContext:Object Reference=Get l_world()) -> (then, RowString:String)
   > Add(TargetArray=Get l_rows() -> (l_rows:Array of Strings), NewItem=@n5) {is=title K2Node_CallArrayFunction}
   n6 = GetCSVHeaders!(execute=@n1.then_1, __WorldContext:Object Reference=Get l_world()) -> (then, RowString:String, Separator:String)
   > Set l_headers(l_headers:String=@n6) -> (Output_Get:String)
   Execute Python Script!(
    execute=@n1.then_2
    PythonScript="import unreal\r\nimport csv\r\nimport os\r\nfrom pathlib import Path\r\nimport subprocess\r\n\r\ndef run(in_headers, in_rows, in_separator):\r\n    file_path = unreal.Paths.project_saved_dir() + \"GART/\"\r\n    file_path =  unreal.Paths.convert_relative_path_to_full(file_path)\r\n    file_path =  os.path.realpath(file_path)\r\n\r\n    final_rows = []\r\n    headers = in_headers.split(\",\")\r\n\r\n    for row in in_rows:\r\n        row = row.replace(\",\",\" \")\r\n        row_values = row.split(in_separator)\r\n        row_dict = {}\r\n        for i in range(len(headers)):\r\n            row_dict[headers[i]] = row_values[i]\r\n        final_rows.append(row_dict)\r\n    \r\n    Path(file_path).mkdir(parents=True, exist_ok=True)\r\n\r\n    file_name = file_path + \"\\GartDataExport.csv\"\r\n    \r\n    with open(file_name, mode='w+', newline='') as csv_file:\r\n        writer = csv.DictWriter(csv_file, fieldnames = headers, extrasaction='ignore')\r\n        writer.writeheader()\t\t\r\n        for each in final_rows:\r\n            writer.writerow(each)\r\n            \r\n    os.startfile(file_path)\r\n\r\nrun(in_headers, in_rows, in_separator)"
    in_headers:String=Get l_headers() -> (l_headers:String)
    in_rows:Array of Strings=Get l_rows()
    in_separator:String=GetCSVSeparator!(__WorldContext:Object Reference=Get l_world()) -> (Separator:String)
   ) -> (then, ReturnValue) {inputs [in_headers,in_rows,in_separator]}
  ]}
 Get_txtNoIssues_Visibility {nodes [
   Get_txtNoIssues_Visibility() {is=title K2Node_FunctionEntry}
   n1 = GetNumItems(self=Get listMain()) -> (ReturnValue=0)
   > Return Node!(execute, ReturnValue:ESlateVisibility Enum=Select(Option 0=Visible, Option 1=Collapsed, Index="integer > integer"(A=@n1, B=1)))
  ]}
 RefreshList {nodes [
   RefreshList() {is=title K2Node_FunctionEntry}
   > ClearListItems(self=Get listMain())
   > n1 = Sequence()
   > n2 = Values!(
    execute
    TargetMap:Map of Names to Struct GART Asset Report List Structures=Get CurrentAssetReportData() -> (CurrentAssetReportData:Map of Names to Struct GART Asset Report List Structures)
   ) -> (then, Values:Array of Struct GART Asset Report List Structures)
   > n3 = For Each Loop(Array:Array of Struct GART Asset Report List Structures=@n2) -> (Array Element:Struct GART Asset Report List Structure)
   > n4 = Construct Object from Class(Class=/Game/NeverCook/EditorUtilities/GenericAssetReviewTool/BP/GART_ListData.GART_ListData_C) -> (ReturnValue:GART List Data Object Reference)
   > Set ReportData!(execute, ReportData:Struct GART Asset Report List Structure=@n3, self:GART List Data Object Reference=@n4) -> (then, Output_Get:Struct GART Asset Report List Structure)
   > AddItem!(execute, self:List View Object Reference=Get listMain(), Item:Object Reference=@n4) -> (then)
   n5 = Values(
    execute=@n1.then_1
    TargetMap:Map of Object References to Struct GART Asset Report List Structures=Get CurrentObjectReportData() -> (CurrentObjectReportData:Map of Object References to Struct GART Asset Report List Structures)
   )
   > n6 = For Each Loop(Array=@n5)
   > n7 = Construct Object from Class(Class=&GART_ListData_C)
   > Set ReportData(ReportData=@n6, self=@n7)
   > AddItem(self=Get listMain(), Item=@n7)
   n8 = Get listMain()
   "SetText (Text)"(
    execute=@n1.then_2
    self=Get txtTotal() -> (txtTotal:Common Text Object Reference)
    InText=Format Text!(Format="LOCTEXT(\"Displaying {count} of entries...\")", count:int=GetNumItems(self=@n8) -> (ReturnValue=0)) -> (Result)
   )
   n9 = GetNumItems(self=Get listMain()) -> (ReturnValue=0)
   > SetVisibility(self=Get txtTotal(), InVisibility=Select(Option 0=Collapsed, Option 1=Visible, Index="integer > integer"(A=@n9, B=1)))
   n10 = GetNumItems(self=Get listMain()) -> (ReturnValue=0)
   > SetVisibility(self=Get txtNoIssues(), InVisibility=Select(Option 0=Visible, Option 1=Collapsed, Index="integer > integer"(A=@n10)))
  ]}
 SaveAssetsToCollection {nodes [
   SaveAssetsToCollection() {is=title K2Node_FunctionEntry}
   > n1 = Sequence()
   > n2 = For Each Loop(Array:Array of Object References=GetListItems(self=Get listMain())) -> (Array Element:Object Reference)
   > n3 = Cast To GART_ListData(Object=@n2)
   n4 = Break Struct GART Asset Report List!(Struct_GART_AssetReportList=Get ReportData(self=@n3)) -> (
    ReportDatas_11_#1
    AssetData_7_#2:Asset Data Structure
    Object_10_#3:Object Reference
   )
   > n5 = For Each Loop(Array:Array of Struct GART Report Output Data Structures=@n4) -> (Array Element:Struct GART Report Output Data Structure)
   n6 = Branch(
    execute=@n5.Completed
    Condition=Is Valid Index!(TargetArray:Array of Asset Data Structures=Get l_assetdatas() -> (l_assetdatas:Array of Asset Data Structures), IndexToTest:int) -> (ReturnValue:bool)
   )
   Return Node!(execute=@n6.else)
   n7 = Is Valid(exec=@n5.LoopBody, InputObject=@n4.Object_10_#3)
   Add(
    execute=@n7.Is Not Valid
    TargetArray:Array of Asset Data Structures=Get l_assetdatas()
    NewItem:"Asset Data Structure (by ref)"=@n4.AssetData_7_#2
   )
   n8 = DoesSaveGameExist!(execute=@n1.then_1, SlotName:String=Get l_saveSlotName() -> (l_saveSlotName:String), UserIndex:int=0) -> (then, ReturnValue:bool)
   > n9 = Branch(Condition=@n8)
   > n10 = LoadGameFromSlot!(execute, SlotName:String=Get l_saveSlotName(), UserIndex:int=0) -> (then, ReturnValue:Save Game Object Reference)
   > n11 = Cast To GART_CollectionResponse!(execute, Object=@n10) -> (then, CastFailed, AsGART Collection Response:GART Collection Response Object Reference) {is=title K2Node_DynamicCast}
   > Set l_object(l_object:GART Collection Response Object Reference=@n11) -> (Output_Get:GART Collection Response Object Reference)
   n12 = Construct Object from Class(execute=@n9.else, Class=/Game/NeverCook/EditorUtilities/GenericAssetReviewTool/BP/Misc/GART_CollectionResponse.GART_CollectionResponse_C) -> (ReturnValue:GART Collection Response Object Reference)
   > Set l_object(l_object=@n12)
   n13 = Show Object Dialog!(
    execute=@n1.then_2
    Title:"Text (by ref)"=MakeLiteralText(Value="INVTEXT(\"Name your collection\")")
    InOutObject:Object Reference=Get l_object() -> (l_object:GART Collection Response Object Reference)
    Options:"Editor Dialog Library Object Details View Options Structure (by ref)"=Make Editor Dialog Library Object Details View Options!(
     bShowObjectName:bool=false
     bAllowSearch:bool=false
     bAllowResizing:bool=False
     MinWidth:int=0
     MinHeight:int=0
     ValueColumnWidthRatio:float=0.6
    ) -> (EditorDialogLibraryObjectDetailsViewOptions:Editor Dialog Library Object Details View Options Structure) {is=title K2Node_MakeStruct}
   ) -> (then, ReturnValue:bool)
   > Branch(Condition=@n13)
   > Set l_name(l_name:Name=Get CollectionName!(self:GART Collection Response Object Reference=Get l_object()) -> (CollectionName:Name=None)) -> (Output_Get:Name=None)
   n14 = Get l_name() -> (l_name:Name=None)
   > Branch(Condition=AND Boolean(A="Not Equal (Name)"(A=@n14) -> (ReturnValue=~), B="Not Equal (Name)"(A=@n14) -> (ReturnValue=~)))
   > SaveGameToSlot!(execute, SaveGameObject:Save Game Object Reference=Get l_object(), SlotName:String=Get l_saveSlotName(), UserIndex:int=0) -> (then, ReturnValue:bool)
   n15 = Get CollectionManagerScriptingSubsystem()
   > n16 = CreateOrEmptyCollection!(
    execute
    self:Collection Manager Subsystem Object Reference=@n15
    Container:Collection Container Source Structure
    Collection:Name=Get l_name() -> (l_name=None)
    ShareType:ECollectionScriptingShareType Enum=Local
   ) -> (then, OutNewOrEmptyCollection:Collection Structure, ReturnValue:bool)
   > AddAssetDatasToCollection!(
    execute
    self:Collection Manager Subsystem Object Reference=@n15
    Collection:"Collection Structure (by ref)"=@n16
    AssetDatas:Array of Asset Data Structures=Get l_assetdatas()
   ) -> (then, ReturnValue:bool)
  ]}
 SetAssetReportData {nodes [
   n1 = SetAssetReportData!() -> (then, ReportData:Map of Names to Struct GART Asset Report List Structures, UpdateList:bool=~) {is=title K2Node_FunctionEntry}
   > Set CurrentAssetReportData(CurrentAssetReportData:Map of Names to Struct GART Asset Report List Structures=@n1) -> (Output_Get:Map of Names to Struct GART Asset Report List Structures)
   > Branch(Condition=@n1.UpdateList)
   > RefreshList() {is=title K2Node_CallFunction}
  ]}
 SetObjectReportData {nodes [
   n1 = SetObjectReportData!() -> (then, ReportData:Map of Object References to Struct GART Asset Report List Structures, UpdateList:bool=~) {is=title K2Node_FunctionEntry}
   > Set CurrentObjectReportData(CurrentObjectReportData:Map of Object References to Struct GART Asset Report List Structures=@n1) -> (Output_Get:Map of Object References to Struct GART Asset Report List Structures)
   > Branch(Condition=@n1.UpdateList)
   > RefreshList()
  ]}
]
variables [
 CurrentAssetReportData {type="map<Struct_GART_AssetReportList>", default {}}
 CurrentObjectReportData {default {}}
]
widgets {
 info {rootWidgetClass=&StackBox, widgetCount=9}
 widgets [
  StackBox_0 {widgetClassPath=&StackBox, widget=StackBox_0, properties {orientation=Orient_Vertical}}
  txtTotal {
   widgetClassPath=&CommonTextBlock
   bIsVariable=true
   parent=StackBox_0
   widget=txtTotal
   properties {
    text=No Issues Found
    visibility=Hidden
    colorAndOpacity {colorUseRule=UseColor_Specified, specifiedColor=<rgba 1 1 1 0.5>}
    font {
     bForceMonospaced=false
     bMaterialIsStencil=false
     fontMaterial=null
     fontObject=&NotoSans
     letterSpacing=0
     monospacedWidth=1
     size=7.5
     skewAmount=0
     typefaceFontName=NotoSans-Regular
     outlineSettings {
      bApplyOutlineToDropShadows=false
      bMiteredCorners=false
      bSeparateFillAlpha=false
      outlineMaterial=null
      outlineSize=0
      outlineColor=<rgba 0 0 0 1>
     }
    }
    shadowOffset=<0 0>
   }
   slot {class=StackBoxSlot, properties {horizontalAlignment=HAlign_Center, verticalAlignment=VAlign_Center}}
  }
  Overlay_0 {
   widgetClassPath=&Overlay
   parent=StackBox_0
   widget=Overlay_0
   slot {class=StackBoxSlot, properties {size {sizeRule=Fill, value=1}}}
  }
  listMain {
   widgetClassPath=&CommonListView
   bIsVariable=true
   parent=Overlay_0
   widget=listMain
   properties {
    entryWidgetClass=&/EUW_GART_ListItem_Main.EUW_GART_ListItem_Main_C
    numDesignerPreviewEntries=0
    scrollBarStyle {
     thickness=8
     draggedThumbImage {
      bIsDynamicallyLoaded=false
      drawAs=RoundedBox
      imageType=NoImage
      mirroring=NoMirror
      resourceName=None
      resourceObject=null
      tiling=NoTile
      imageSize=<0 0>
      margin=<ltrb 0 0 0 0>
      outlineSettings {
       bUseBrushTransparency=false
       roundingType=FixedRadius
       width=0
       color {colorUseRule=UseColor_Specified, specifiedColor=<rgba 0 0 0 0>}
       cornerRadii=<4 4 4 4>
      }
      tintColor {colorUseRule=UseColor_Specified, specifiedColor=<rgba 0.2159 0.2159 0.2159 1>}
      uVRegion {bIsValid=false, max=<0 0>, min=<0 0>}
     }
     horizontalBackgroundImage {
      bIsDynamicallyLoaded=false
      drawAs=NoDrawType
      imageType=NoImage
      mirroring=NoMirror
      resourceName=None
      resourceObject=null
      tiling=NoTile
      imageSize=<0 0>
      margin=<ltrb 0 0 0 0>
      outlineSettings {
       bUseBrushTransparency=false
       roundingType=HalfHeightRadius
       width=0
       color {colorUseRule=UseColor_Specified, specifiedColor=<rgba 0 0 0 0>}
       cornerRadii=<0 0 0 0>
      }
      tintColor {colorUseRule=UseColor_Specified, specifiedColor=<rgba 1 1 1 1>}
      uVRegion {bIsValid=false, max=<0 0>, min=<0 0>}
     }
     horizontalBottomSlotImage {
      bIsDynamicallyLoaded=false
      drawAs=NoDrawType
      imageType=NoImage
      mirroring=NoMirror
      resourceName=None
      resourceObject=null
      tiling=NoTile
      imageSize=<0 0>
      margin=<ltrb 0 0 0 0>
      outlineSettings {
       bUseBrushTransparency=false
       roundingType=HalfHeightRadius
       width=0
       color {colorUseRule=UseColor_Specified, specifiedColor=<rgba 0 0 0 0>}
       cornerRadii=<0 0 0 0>
      }
      tintColor {colorUseRule=UseColor_Specified, specifiedColor=<rgba 1 1 1 1>}
      uVRegion {bIsValid=false, max=<0 0>, min=<0 0>}
     }
     horizontalTopSlotImage {
      bIsDynamicallyLoaded=false
      drawAs=NoDrawType
      imageType=NoImage
      mirroring=NoMirror
      resourceName=None
      resourceObject=null
      tiling=NoTile
      imageSize=<0 0>
      margin=<ltrb 0 0 0 0>
      outlineSettings {
       bUseBrushTransparency=false
       roundingType=HalfHeightRadius
       width=0
       color {colorUseRule=UseColor_Specified, specifiedColor=<rgba 0 0 0 0>}
       cornerRadii=<0 0 0 0>
      }
      tintColor {colorUseRule=UseColor_Specified, specifiedColor=<rgba 1 1 1 1>}
      uVRegion {bIsValid=false, max=<0 0>, min=<0 0>}
     }
     hoveredThumbImage {
      bIsDynamicallyLoaded=false
      drawAs=RoundedBox
      imageType=NoImage
      mirroring=NoMirror
      resourceName=None
      resourceObject=null
      tiling=NoTile
      imageSize=<0 0>
      margin=<ltrb 0 0 0 0>
      outlineSettings {
       bUseBrushTransparency=false
       roundingType=FixedRadius
       width=0
       color {colorUseRule=UseColor_Specified, specifiedColor=<rgba 0 0 0 0>}
       cornerRadii=<4 4 4 4>
      }
      tintColor {colorUseRule=UseColor_Specified, specifiedColor=<rgba 0.2159 0.2159 0.2159 1>}
      uVRegion {bIsValid=false, max=<0 0>, min=<0 0>}
     }
     normalThumbImage {
      bIsDynamicallyLoaded=false
      drawAs=RoundedBox
      imageType=NoImage
      mirroring=NoMirror
      resourceName=None
      resourceObject=null
      tiling=NoTile
      imageSize=<0 0>
      margin=<ltrb 0 0 0 0>
      outlineSettings {
       bUseBrushTransparency=false
       roundingType=FixedRadius
       width=0
       color {colorUseRule=UseColor_Specified, specifiedColor=<rgba 0 0 0 0>}
       cornerRadii=<4 4 4 4>
      }
      tintColor {colorUseRule=UseColor_Specified, specifiedColor=<rgba 0.0953 0.0953 0.0953 1>}
      uVRegion {bIsValid=false, max=<0 0>, min=<0 0>}
     }
     verticalBackgroundImage {
      bIsDynamicallyLoaded=false
      drawAs=NoDrawType
      imageType=NoImage
      mirroring=NoMirror
      resourceName=None
      resourceObject=null
      tiling=NoTile
      imageSize=<0 0>
      margin=<ltrb 0 0 0 0>
      outlineSettings {
       bUseBrushTransparency=false
       roundingType=HalfHeightRadius
       width=0
       color {colorUseRule=UseColor_Specified, specifiedColor=<rgba 0 0 0 0>}
       cornerRadii=<0 0 0 0>
      }
      tintColor {colorUseRule=UseColor_Specified, specifiedColor=<rgba 1 1 1 1>}
      uVRegion {bIsValid=false, max=<0 0>, min=<0 0>}
     }
     verticalBottomSlotImage {
      bIsDynamicallyLoaded=false
      drawAs=NoDrawType
      imageType=NoImage
      mirroring=NoMirror
      resourceName=None
      resourceObject=null
      tiling=NoTile
      imageSize=<0 0>
      margin=<ltrb 0 0 0 0>
      outlineSettings {
       bUseBrushTransparency=false
       roundingType=HalfHeightRadius
       width=0
       color {colorUseRule=UseColor_Specified, specifiedColor=<rgba 0 0 0 0>}
       cornerRadii=<0 0 0 0>
      }
      tintColor {colorUseRule=UseColor_Specified, specifiedColor=<rgba 1 1 1 1>}
      uVRegion {bIsValid=false, max=<0 0>, min=<0 0>}
     }
     verticalTopSlotImage {
      bIsDynamicallyLoaded=false
      drawAs=NoDrawType
      imageType=NoImage
      mirroring=NoMirror
      resourceName=None
      resourceObject=null
      tiling=NoTile
      imageSize=<0 0>
      margin=<ltrb 0 0 0 0>
      outlineSettings {
       bUseBrushTransparency=false
       roundingType=HalfHeightRadius
       width=0
       color {colorUseRule=UseColor_Specified, specifiedColor=<rgba 0 0 0 0>}
       cornerRadii=<0 0 0 0>
      }
      tintColor {colorUseRule=UseColor_Specified, specifiedColor=<rgba 1 1 1 1>}
      uVRegion {bIsValid=false, max=<0 0>, min=<0 0>}
     }
    }
   }
   slot {class=OverlaySlot, properties {horizontalAlignment=HAlign_Fill, verticalAlignment=VAlign_Fill}}
  }
  ScaleBox_0 {
   widgetClassPath=&ScaleBox
   parent=Overlay_0
   widget=ScaleBox_0
   slot {class=OverlaySlot, properties {horizontalAlignment=HAlign_Fill, verticalAlignment=VAlign_Fill, padding=<ltrb 15 15 15 15>}}
  }
  txtNoIssues {
   widgetClassPath=&CommonTextBlock
   bIsVariable=true
   parent=ScaleBox_0
   widget=txtNoIssues
   properties {
    text=No Issues Found
    colorAndOpacity {colorUseRule=UseColor_Specified, specifiedColor=<rgba 1 1 1 0.5>}
    font {
     bForceMonospaced=false
     bMaterialIsStencil=false
     fontMaterial=null
     fontObject=&NotoSans
     letterSpacing=0
     monospacedWidth=1
     size=24
     skewAmount=0
     typefaceFontName=NotoSans-Regular
     outlineSettings {
      bApplyOutlineToDropShadows=false
      bMiteredCorners=false
      bSeparateFillAlpha=false
      outlineMaterial=null
      outlineSize=0
      outlineColor=<rgba 0 0 0 1>
     }
    }
    shadowOffset=<0 0>
   }
   slot {class=ScaleBoxSlot}
  }
  panelExportButtons {widgetClassPath=&HorizontalBox, bIsVariable=true, parent=StackBox_0, widget=panelExportButtons, slot {class=StackBoxSlot}}
  btnExport {
   widgetClassPath=&TA_Generic01_Button
   parent=panelExportButtons
   widget=btnExport
   properties {
    label=Export To CSV
    toolTipText="Export all report data to a CSV located in: \\FortniteGame\\Saved\\GART\\GartDataExport.csv\r\n(Should open File Explorer to the location)"
    visibility=Visible
    buttonColor_Default=<rgba 0.5365 0.417 0.1794 1>
   }
   slot {class=HorizontalBoxSlot, properties {size {sizeRule=Fill, value=1}}}
  }
  btnCollection {
   widgetClassPath=&TA_Generic01_Button
   parent=panelExportButtons
   widget=btnCollection
   properties {
    label=Save Assets To Collection
    toolTipText="Save Assets in the list to a collection of a name you choose.  \r\n(Does not work on object reports)"
    visibility=Visible
    buttonColor_Default=<rgba 0.5365 0.417 0.1794 1>
   }
   slot {class=HorizontalBoxSlot, properties {size {sizeRule=Fill, value=1}}}
  }
 ]
}
